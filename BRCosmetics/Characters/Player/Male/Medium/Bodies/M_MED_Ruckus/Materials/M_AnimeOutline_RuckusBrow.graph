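# Translucent unlit outline material for the Ruckus character that draws anime-style inner lines over a slightly normal-inflated mesh.
# Four offset SceneDepth taps, optionally combined with roughness GBuffer edges, drive opacity, with line width and sensitivity scaling by distance.
# Color is either darkened scene diffuse or a parameter color fading to a brow tint via a pre-skinned-position mask and distance.
# Forward and mobile paths use their own color and can disable lines.
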Material /BRCosmetics/Characters/Player/Male/Medium/Bodies/M_MED_Ruckus/Materials/M_AnimeOutline_RuckusBrow
graph {
 properties {
  bEnableResponsiveAA=true
  bUseMaterialAttributes=true
  bUsedWithMeshDeformer=true
  bUsedWithMorphTargets=true
  bUsedWithSkeletalMesh=true
  blendMode=BLEND_Translucent
  pixelDepthOffsetMode=PDOM_Legacy
  shadingModel=MSM_Unlit
  twoSided=true
  usedShadingModels=Unlit
 }
 nodes [
  Clamp(None=Divide(A=PixelDepth().R, B=ScalarParameter() {defaultValue=1024, parameterName=LineScaleDistance}))
  Constant() {r=2}
  Constant() {r=0.75}
  Constant()
  LinearInterpolate(
   A=ScalarParameter() {defaultValue=1.3, group=Cartoon, parameterName=CRTNO_Outline-InnerSize1_}
   B=ScalarParameter() {defaultValue=0.5, group=Cartoon, parameterName=CRTNO_Outline-InnerSize1_Brow}
   Alpha=SphereMask(
    A=TextureSample() {texture=/Game/Characters/Player/Male/Medium/Bodies/M_MED_Ruckus/Textures/T_M_MED_Ruckus_Head_D}.RGB
    B=Constant3Vector() {constant=<rgba 0.0532 0.2551 0.5104 0>}
   )
  ) {constB=0}
  n1 = ScalarParameter() {defaultValue=16, parameterName=SampleBias_RGH}
  n2 = PixelDepth()
  n3 = Divide(
   A=Subtract(A=@n2.R, B=ScalarParameter() {defaultValue=50, group=Cartoon, parameterName=CRTNO_InnerLine_FadeStart_RGH})
   B=ScalarParameter() {defaultValue=500, group=Cartoon, parameterName=CRTNO_InnerLine_FadeDistance_RGH}
  )
  n4 = PreSkinnedPosition()
  n5 = Divide(A=Add(A=VertexInterpolator(VS=@n4).PS, B=Constant3Vector() {constant=<rgba 0 0 -150 0>}), B=PreSkinnedLocalBounds().Max)
  n6 = SphereMask(A=ComponentMask(None=Normalize(VectorInput=@n5)) {b=true}, B=Constant() {r=0.17}, Radius=Constant() {r=0.023}, Hardness=Constant() {r=1})
  n7 = Saturate(None=Multiply(A=Saturate(None=@n3), B=Saturate(None=@n6)))
  LinearInterpolate(A=ScalarParameter() {defaultValue=16, parameterName=SampleBias_Min_RGH}, B=@n1, Alpha=@n7)
  LinearInterpolate(
   A=ScalarParameter() {defaultValue=64, parameterName=Scale_RGH_Near}
   B=ScalarParameter() {defaultValue=1, parameterName=Scale_RGH_Far}
   Alpha=@n7
  )
  n8 = ScalarParameter() {defaultValue=3, group=Cartoon, parameterName=CRTNO_Outline-InnerMaskBias}
  LinearInterpolate(A=@n8, B=ScalarParameter() {defaultValue=3, group=Cartoon, parameterName=CRTNO_Outline-InnerMaskBiasRGH}, Alpha=@n7)
  n9 = VectorParameter() {parameterName=OutlineColor, defaultValue=<rgba 0 0 0 1>}
  n10 = VectorParameter() {parameterName=OutlineColorBrowFade, defaultValue=<rgba 0.1015 0.2286 0.4427 1>}
  n11 = ComponentMask(None=SceneTexture() {sceneTextureId=PPI_DiffuseColor}.Color) {b=true, g=true, r=true}
  n12 = VectorParameter() {parameterName=OutlineDakenMultiplier, defaultValue=<rgba 0.1 0.1 0.1 0>}
  n13 = VectorParameter() {parameterName=MobileOutlineColor}
  n14 = ShadingPathSwitch(
   Default=StaticSwitchParameter(True=LinearInterpolate(A=@n9.RGB, B=@n10.RGB, Alpha=@n7), False=Multiply(A=@n11, B=@n12.RGB)) {parameterName=UseOutlineColor}
   Forward=@n13.RGB
   Mobile=@n13.RGB
  )
  n15 = PixelDepth()
  n16 = SceneDepth() {desc=SceneDepth works on mobile whereas SceneTexture doesn't}
  n17 = Divide(A=ObjectRadius(), B=ScalarParameter() {defaultValue=97, parameterName=ObjectRadius})
  n18 = Clamp(None=Add(A=Subtract(A=@n15.R, B=@n16.R), B=Multiply(A=@n8, B=@n17)))
  n19 = ScalarParameter() {defaultValue=50, group=Cartoon, parameterName=CRTNO_InnerLine_FadeStart}
  n20 = Clamp(None=Divide(A=Subtract(A=@n2.R, B=@n19), B=ScalarParameter() {defaultValue=500, group=Cartoon, parameterName=CRTNO_InnerLine_FadeDistance}))
  n21 = ShadingPathSwitch(
   Default=LinearInterpolate(
    A=ScalarParameter() {defaultValue=16, parameterName=SampleBias_Min}
    B=ScalarParameter() {defaultValue=16, parameterName=SampleBias}
    Alpha=@n20
   )
   Mobile=LinearInterpolate(
    A=ScalarParameter() {defaultValue=1, parameterName=MobileSampleBias_Min}
    B=ScalarParameter() {defaultValue=16, parameterName=Mobile_SampleBias}
    Alpha=@n20
   )
  )
  n22 = ComponentMask(None=ScreenPosition().ViewportUV) {g=true, r=true}
  n23 = Subtract(A=Divide(A=ViewSize(), B=Constant() {r=320}), B=Constant() {r=2}) {constB=0}
  n24 = Subtract(A=PixelDepth().R, B=ScalarParameter() {defaultValue=50, group=Cartoon, parameterName=CRTNO_Outline-InnerScaleStart})
  n25 = ScalarParameter() {defaultValue=500, group=Cartoon, parameterName=CRTNO_Outline-InnerScaleDistance}
  n26 = LinearInterpolate(
   A=ScalarParameter() {defaultValue=2, group=Cartoon, parameterName=CRTNO_Outline-InnerSize1}
   B=ScalarParameter() {defaultValue=0.75, group=Cartoon, parameterName=CRTNO_Outline-InnerSize2}
   Alpha=Clamp(None=Divide(A=@n24, B=@n25))
  )
  n27 = Multiply(A=@n26, B=ScalarParameter() {defaultValue=1.5, group=Cartoon, parameterName=CRTNO_Outline-InnerSizeMobileMultiplier})
  n28 = Multiply(
   A=Clamp(None=Ceil(None=@n23)) {maxDefault=2, minDefault=1}
   B=Multiply(A=SceneTexelSize(), B=ShadingPathSwitch(Default=@n26, Forward=@n27, Mobile=@n27))
  ) {constA=2}
  n29 = Constant()
  n30 = AppendVector(A=ComponentMask(None=@n28) {desc=x, r=true}, B=@n29)
  n31 = Add(A=@n22, B=@n30) {desc=x+}
  n32 = AppendVector(A=@n29, B=ComponentMask(None=@n28) {desc=y, g=true}) {desc=y+}
  n33 = Add(A=@n22, B=@n32) {desc=y+}
  n34 = Constant() {r=-1}
  n35 = Add(A=@n22, B=Multiply(A=@n30, B=@n34)) {desc=x-}
  n36 = Add(A=@n22, B=Multiply(A=@n34, B=@n32)) {desc=y-}
  n37 = LinearInterpolate(
   A=ScalarParameter() {defaultValue=-1, parameterName=Divisor_Min}
   B=ScalarParameter() {defaultValue=-16, parameterName=Divisor_max}
   Alpha=@n20
  )
  n38 = LinearInterpolate(
   A=ScalarParameter() {defaultValue=-1, parameterName=MobileDivisor_Min}
   B=ScalarParameter() {defaultValue=-16, parameterName=MobileDivisor_max}
   Alpha=@n20
  )
  n39 = ComponentMask(None=MaterialFunctionCall!(
    Base=@n16.R
    BaseSampleBias=Multiply(A=@n17, B=@n21)
    Scale=ScalarParameter() {defaultValue=4, parameterName=Scale}
    Sample1=SceneDepth(UVs=@n31).R
    Sample2=SceneDepth(UVs=@n33).R
    Sample3=SceneDepth(UVs=@n35).R
    Sample4=SceneDepth(UVs=@n36).R
    Divisor=ShadingPathSwitch(Default=@n37, Mobile=@n38)
    UseCenteredLines
    UseRGB
   ) {materialFunction=/Game/Packages/jl_tests/combine4Samples}.Result) {r=true}
  n40 = ComponentMask(None=MaterialFunctionCall(
    Base=SceneTexture() {sceneTextureId=PPI_Roughness}.Color
    BaseSampleBias=@n1
    Scale=Constant() {r=4}
    Sample1=SceneTexture(UVs=@n31) {sceneTextureId=PPI_Roughness}.Color
    Sample2=SceneTexture(UVs=@n33) {sceneTextureId=PPI_Roughness}.Color
    Sample3=SceneTexture(UVs=@n35) {sceneTextureId=PPI_Roughness}.Color
    Sample4=SceneTexture(UVs=@n36) {sceneTextureId=PPI_Roughness}.Color
    Divisor=ScalarParameter() {defaultValue=-0.04, group=RoughnessGBufferLines, parameterName=InnerLinesDivisor}
   ) {materialFunction=&combine4Samples}.Result) {r=true}
  n41 = StaticSwitchParameter(True=ShadingPathSwitch(Default=Max(A=@n39, B=@n40), Forward=@n39, Mobile=@n39), False=@n39) {group=RoughnessGBufferLines, parameterName=UseRoughnessGBufferOutlines}
  n42 = Multiply(A=Saturate(None=Multiply(A=@n18, B=@n41)), B=ScalarParameter() {defaultValue=1, group=Cartoon, parameterName=CRTNO_Outline-InnerAmount})
  n43 = Constant()
  n44 = StaticSwitchParameter(True=@n42, False=ShadingPathSwitch(Default=@n42, Forward=@n43, Mobile=@n43)) {group=Cartoon, parameterName=UseDepthBufferOutlinesOnMobile}
  n45 = Multiply(A=VertexNormalWS(), B=ScalarParameter() {defaultValue=0.5, parameterName="ScaleMesh\\"})
  MaterialFunctionCall(In=SetMaterialAttributes!(
    MaterialAttributes
    Emissive Color=NamedRerouteDeclaration(None=@n14) {name=emissive, nodeColor=<rgba 1 0.5391 0 1>}
    Opacity=NamedRerouteDeclaration(None=@n44) {name=opacity, nodeColor=<rgba 0 1 0.6719 1>}
    World Position Offset=NamedRerouteDeclaration(None=@n45) {name=wpo, nodeColor=<rgba 0.2656 0 1 1>}
   ) {attributeSetTypes [#1, #2, #3]}) {materialFunction=&MF_TechArtOverrides}
  n46 = SceneTexture() {sceneTextureId=PPI_Roughness}
  Multiply(
   A=@n42
   B=SphereMask(
    A=ComponentMask(None=@n46.Color) {b=true, g=true, r=true}
    B=VectorParameter() {parameterName=SS_MaskColor, defaultValue=<rgba 0.5 0.5 0.5 1>}.RGB
    Radius=Constant() {r=0.01}
    Hardness=Constant() {r=1}
   )
  )
  Reroute()
  n47 = Distance()
  n48 = ScalarParameter() {defaultValue=6000, group=NeonDepth, parameterName=CheekLine__FarDistance}
  n49 = ScalarParameter() {defaultValue=5000, group=NeonDepth, parameterName=CheekLine_NearDistance}
  Saturate(None=Divide(A=Subtract(A=@n47, B=@n48), B=Subtract(A=@n49, B=@n48)))
 ]
}
