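# Material function for the Ruckus character that adds a dark Fresnel rim outline on the mobile shading path, passing other paths' material attributes through unchanged.
# Inverted pixel-normal Fresnel, raised to a depth-scaled exponent, blends two outline constants and is saturated to multiply base color and emissive.
# Vertex color red masks the outline, with static switches to invert it or apply a UV sphere-mask eyelid override.
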MaterialFunction /Game/Characters/Player/Male/Medium/Bodies/M_MED_Ruckus/Materials/MF_Ruckus_MobileOutline
graph {nodes [
  n1 = GetMaterialAttributes(None=FunctionInput() {inputName=Material, inputType=FunctionInput_MaterialAttributes}) {attributeGetTypes [
    #1
    #2
    #3
    #4
   ]}
  n2 = ScalarParameter() {defaultValue=-4, group=HoloFoil, parameterName=BL_OutlineC1}
  n3 = ScalarParameter() {defaultValue=12, group=HoloFoil, parameterName=BL_OutlineC2}
  n4 = ScalarParameter() {defaultValue=6, group=HoloFoil, parameterName=BL_OutlineExponent1}
  n5 = PixelNormalWS()
  n6 = PixelDepth()
  n7 = ScalarParameter() {defaultValue=0.01, parameterName=BL_OutlineExpMult}
  n8 = Power(
   Base=OneMinus(None=Fresnel(ExponentIn=@n4, Normal=@n5) {baseReflectFraction=0})
   Exp=Multiply(A=ScalarParameter() {defaultValue=12, group=HoloFoil, parameterName=BL_OutlineExponent2}, B=Multiply(A=@n6.R, B=@n7) {constB=0.002})
  )
  n9 = VertexColor()
  n10 = TextureCoordinate()
  n11 = Constant2Vector() {g=1, r=1}
  n12 = Constant() {r=0.7}
  n13 = StaticSwitchParameter(True=LinearInterpolate(A=@n9.R, Alpha=SphereMask(A=@n10, B=@n11, Radius=@n12)), False=@n9.R) {parameterName=Morty Eyelid Override}
  n14 = OneMinus(None=@n13)
  n15 = Saturate(None=LinearInterpolate(A=LinearInterpolate(A=@n2, B=@n3, Alpha=@n8), Alpha=StaticSwitchParameter(True=@n14, False=@n13) {parameterName=InvertVertColor}))
  n16 = SetMaterialAttributes!(MaterialAttributes=@n1.MaterialAttributes, BaseColor=Multiply(A=@n1.BaseColor, B=@n15), EmissiveColor=Multiply(A=@n1.EmissiveColor, B=@n15)) {attributeSetTypes [#2, #3]}
  FunctionOutput(None=SetMaterialAttributes!(MaterialAttributes=ShadingPathSwitch(Default=@n1.MaterialAttributes, Deferred=@n1.MaterialAttributes, Mobile=@n16) {desc=Migrated from FeatureLevelSwitch}))
  ShadingPathSwitch(Default=@n1.MaterialAttributes, Deferred=@n1.MaterialAttributes, Forward=@n1.MaterialAttributes, Mobile=@n16)
 ]}
