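# Master cloth material for the Ruckus player character, adding anime-style shading between the shared character PreFX and PostFX functions.
# The PreFX result runs through anime shading, an optional pickaxe-derived eye movement function switched by EyeMovement, then a mobile outline function.
# The RuckusMiniGlassSwap switch replaces world position offset with one that collapses vertex-green-masked vertices to the object position according to a rounded GlassBroken parameter, hiding them.
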Material /BRCosmetics/Characters/Player/Male/Medium/Bodies/M_MED_Ruckus/Materials/M_Ruckus_MASTER
graph {
 properties {
  bAutomaticallySetUsageInEditor=false
  bUseMaterialAttributes=true
  bUsedWithClothing=true
  bUsedWithMeshDeformer=true
  bUsedWithMorphTargets=true
  bUsedWithSkeletalMesh=true
  pixelDepthOffsetMode=PDOM_Legacy
  shadingModel=MSM_Cloth
  usedShadingModels=Cloth
 }
 nodes [
  n1 = MaterialFunctionCall() {materialFunction=&BaseFN_CharacterFunction_PreFX}
  n2 = MaterialFunctionCall!(Material=@n1.Result, DF InkLine Visibility Mask, DF Inkline Injection, DF Inkline Color Override, DF Inkline Color Override Mask) {materialFunction=/Game/Characters/Player/Male/Medium/Bodies/M_MED_Ruckus/Materials/MF_RuckusAnimeShading}
  n3 = MaterialFunctionCall!(LookDistanceOverride, In=@n2.Result) {materialFunction=/BRCosmetics/Weapons/FORT_Melee/Pickaxe_Ruckus/Materials/MF_Ruckus_Pickaxe_Eyes}
  n4 = GetMaterialAttributes(None=MaterialFunctionCall!(Material=StaticSwitchParameter(True=@n3.Result, False=@n2.Result) {group=Ruckus, parameterName=EyeMovement}) {materialFunction=/Game/Characters/Player/Male/Medium/Bodies/M_MED_Ruckus/Materials/MF_Ruckus_MobileOutline}.Result) {attributeGetTypes [#1]}
  n5 = WorldPosition()
  n6 = ObjectPositionWS()
  n7 = OneMinus(None=VertexColor().G)
  n8 = Round(None=Saturate(None=ScalarParameter() {parameterName=GlassBroken}))
  n9 = Multiply(A=Multiply(A=Subtract(A=@n5.XYZ, B=@n6)) {constB=-1}, B=LinearInterpolate(A=Ceil(None=@n7), B=Constant(), Alpha=OneMinus(None=@n8)))
  MaterialFunctionCall(
   AO=@n1.AO
   Distance Blend=@n1.AO
   Fresnel=@n1.AO
   Gmap=@n1.AO
   Pre Skinned Local Position=@n1.AO
   Material=SetMaterialAttributes!(
    MaterialAttributes=@n4.MaterialAttributes
    World Position Offset=StaticSwitchParameter(True=@n9, False=@n4.World Position Offset) {group=Ruckus, parameterName=RuckusMiniGlassSwap}
   ) {attributeSetTypes [#1]}
  ) {materialFunction=&BaseFN_CharacterFunction_PostFX}
 ]
}
