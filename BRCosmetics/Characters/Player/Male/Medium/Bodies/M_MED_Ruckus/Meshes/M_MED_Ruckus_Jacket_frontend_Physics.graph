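# Physics asset for the frontend variant of the male medium Ruckus jacket, simulating the coat's hanging panels around the lower body.
# Front, middle and back bone chains on each side use box segments with capsule tips, hanging from tiny base bodies joined to the pelvis.
# Cross constraints tie neighboring lower segments and the two back halves into a skirt-like lattice.
# Pelvis, lower spine, thigh and calf shapes supply the body volumes.
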PhysicsAsset /BRCosmetics/Characters/Player/Male/Medium/Bodies/M_MED_Ruckus/Meshes/M_MED_Ruckus_Jacket_frontend_Physics
bodies {
 calf_l [{shapeType=Capsule, length=37.0498, radius=7.3737, center=<-22.8751 -0.8932 0.0072>, rotation=<rot -0.1081 88.8284 -89.3868>}]
 calf_r [
  {shapeType=Sphere, radius=11.25, center=<48.5391 -5.6785 0.0001>}
  {shapeType=Capsule, length=37.0498, radius=7.3737, center=<21.9799 0.4078 -0.0157>, rotation=<rot -0.0543 83.8285 -89.3797>}
 ]
 jacket_bk_1_l [{shapeType=Box, extentX=16, extentY=4, extentZ=16, center=<8 -1 0>, rotation=<rot 0 0 15.0008>}]
 jacket_bk_1_r [{shapeType=Box, extentX=16, extentY=4, extentZ=16, center=<8 -1 0>, rotation=<rot 0 0 -15.001>}]
 jacket_bk_2_l [{shapeType=Box, extentX=16, extentY=4, extentZ=16, center=<8 -1 0>, rotation=<rot 0 0 10.0007>}]
 jacket_bk_2_r [{shapeType=Box, extentX=16, extentY=4, extentZ=16, center=<8 -1 0>, rotation=<rot 0 0 -10.0003>}]
 jacket_bk_3_l [{shapeType=Box, extentX=16, extentY=4, extentZ=16, center=<8 -1 0>}]
 jacket_bk_3_r [{shapeType=Box, extentX=16, extentY=4, extentZ=16, center=<8 -1 0>}]
 jacket_bk_4_l [{shapeType=Capsule, length=18.75, radius=4, center=<4.7063 -1 -0.0002>}]
 jacket_bk_4_r [{shapeType=Capsule, length=18.75, radius=4, center=<4.7063 -1 -0.0002>}]
 jacket_bk_base_l [{shapeType=Capsule, length=0.505, radius=0.505}]
 jacket_bk_base_r [{shapeType=Capsule, length=0.505, radius=0.505}]
 jacket_fr_1_l [{shapeType=Box, extentX=10, extentY=4, extentZ=10, center=<5 -1 -3>}]
 jacket_fr_1_r [{shapeType=Box, extentX=10, extentY=4, extentZ=12, center=<5 -1 3>}]
 jacket_fr_1b_l [{shapeType=Box, extentX=10, extentY=4, extentZ=9.5, center=<5 -1 -3>}]
 jacket_fr_1b_r [{shapeType=Box, extentX=10, extentY=4, extentZ=12, center=<5 -1 3>}]
 jacket_fr_2_l [{shapeType=Box, extentX=24, extentY=4, extentZ=12, center=<12 -1 0>}]
 jacket_fr_2_r [{shapeType=Box, extentX=24, extentY=4, extentZ=12, center=<12 -1 0>}]
 jacket_fr_3_l [{shapeType=Box, extentX=16, extentY=4, extentZ=12, center=<8 -1 0>}]
 jacket_fr_3_r [{shapeType=Box, extentX=16, extentY=4, extentZ=12, center=<8 -1 0>}]
 jacket_fr_4_l [{shapeType=Capsule, length=10.25, radius=4, center=<4.7063 -1 -0.0002>}]
 jacket_fr_4_r [{shapeType=Capsule, length=10.25, radius=4, center=<4.7063 -1 -0.0002>}]
 jacket_fr_base_l [{shapeType=Capsule, length=0.505, radius=0.505}]
 jacket_fr_base_r [{shapeType=Capsule, length=0.505, radius=0.505}]
 jacket_mid_1_l [{shapeType=Box, extentX=16, extentY=4, extentZ=16, center=<8 -1 0>, rotation=<rot 0 0 -15.0015>}]
 jacket_mid_1_r [{shapeType=Box, extentX=16, extentY=4, extentZ=16, center=<8 -1 0>, rotation=<rot 0 0 14.9995>}]
 jacket_mid_2_l [{shapeType=Box, extentX=16, extentY=4, extentZ=16, center=<8 -1 0>}]
 jacket_mid_2_r [{shapeType=Box, extentX=16, extentY=4, extentZ=16, center=<8 -1 0>}]
 jacket_mid_3_l [{shapeType=Box, extentX=16, extentY=4, extentZ=19, center=<8 -1 0>, rotation=<rot 0 0 -25.0012>}]
 jacket_mid_3_r [{shapeType=Box, extentX=16, extentY=4, extentZ=19, center=<8 -1 0>, rotation=<rot 0 0 25.0011>}]
 jacket_mid_4_l [{shapeType=Capsule, length=16.75, radius=4, center=<4.7063 -1 -0.0002>, rotation=<rot 0 0 -20.0002>}]
 jacket_mid_4_r [{shapeType=Capsule, length=14, radius=4, center=<4.7063 -1 -0.0002>, rotation=<rot 0 0 20.0015>}]
 jacket_mid_base_l [{shapeType=Capsule, length=0.505, radius=0.505}]
 jacket_mid_base_r [{shapeType=Capsule, length=0.505, radius=0.505}]
 pelvis [{shapeType=Capsule, length=10, radius=9.25, center=<-5.4242 -1.0391 -0.0002>, rotation=<rot 0 -90.0002 -0.004>}]
 root [{shapeType=Capsule, length=0.505, radius=0.505, rotation=<rot 0 -90 0>}]
 spine_01 [{shapeType=Capsule, length=7.377, radius=10.75, rotation=<rot 0 -90 -90>}]
 thigh_l [{shapeType=Capsule, length=30, radius=7, center=<-24 -1.2 0>, rotation=<rot 0 94.1316 -89.5811>}]
 thigh_r [{shapeType=Capsule, length=30, radius=7, center=<24 1.2 0>, rotation=<rot 0 -86.0963 -88.5372>}]
}
constraints [
 {bone1Name=thigh_l, bone2Name=pelvis, swing1LimitDegrees=45, swing2LimitDegrees=45, twistLimitDegrees=45}
 {bone1Name=calf_l, bone2Name=thigh_l, swing1LimitDegrees=45, swing2LimitDegrees=45, twistLimitDegrees=45}
 {bone1Name=thigh_r, bone2Name=pelvis, swing1LimitDegrees=45, swing2LimitDegrees=45, twistLimitDegrees=45}
 {bone1Name=calf_r, bone2Name=thigh_r, swing1LimitDegrees=45, swing2LimitDegrees=45, twistLimitDegrees=45}
 {bone1Name=jacket_fr_2_r, bone2Name=jacket_fr_1b_r, swing1LimitDegrees=60, swing2LimitDegrees=60, twistLimitDegrees=10}
 {bone1Name=jacket_bk_1_r, bone2Name=jacket_bk_base_r, swing1LimitDegrees=45, swing2LimitDegrees=45, twistLimitDegrees=10}
 {bone1Name=jacket_bk_3_r, bone2Name=jacket_bk_2_r, swing1LimitDegrees=60, swing2LimitDegrees=60, twistLimitDegrees=10}
 {bone1Name=jacket_bk_1_l, bone2Name=jacket_bk_base_l, swing1LimitDegrees=45, swing2LimitDegrees=45, twistLimitDegrees=10}
 {bone1Name=jacket_bk_3_l, bone2Name=jacket_bk_2_l, swing1LimitDegrees=60, swing2LimitDegrees=60, twistLimitDegrees=10}
 {bone1Name=jacket_fr_2_l, bone2Name=jacket_fr_1b_l, swing1LimitDegrees=60, swing2LimitDegrees=60, twistLimitDegrees=10}
 {bone1Name=jacket_fr_3_l, bone2Name=jacket_fr_2_l, swing1LimitDegrees=60, swing2LimitDegrees=60, twistLimitDegrees=10}
 {bone1Name=jacket_fr_4_l, bone2Name=jacket_fr_3_l, swing1LimitDegrees=40, swing2LimitDegrees=20, twistLimitDegrees=10}
 {bone1Name=jacket_fr_1_l, bone2Name=jacket_fr_base_l, swing1LimitDegrees=45, swing2LimitDegrees=45, twistLimitDegrees=10}
 {bone1Name=jacket_bk_4_l, bone2Name=jacket_bk_3_l, swing1LimitDegrees=40, swing2LimitDegrees=20, twistLimitDegrees=10}
 {bone1Name=jacket_bk_2_l, bone2Name=jacket_bk_1_l, swing1LimitDegrees=60, swing2LimitDegrees=60, twistLimitDegrees=10}
 {bone1Name=jacket_bk_2_r, bone2Name=jacket_bk_1_r, swing1LimitDegrees=60, swing2LimitDegrees=60, twistLimitDegrees=10}
 {bone1Name=jacket_bk_4_r, bone2Name=jacket_bk_3_r, swing1LimitDegrees=40, swing2LimitDegrees=20, twistLimitDegrees=10}
 {bone1Name=jacket_fr_4_r, bone2Name=jacket_fr_3_r, swing1LimitDegrees=40, swing2LimitDegrees=20, twistLimitDegrees=10}
 {bone1Name=jacket_fr_1_r, bone2Name=jacket_fr_base_r, swing1LimitDegrees=45, swing2LimitDegrees=45, twistLimitDegrees=10}
 {bone1Name=jacket_fr_3_r, bone2Name=jacket_fr_2_r, swing1LimitDegrees=60, swing2LimitDegrees=60, twistLimitDegrees=10}
 {
  bone1Name=pelvis
  bone2Name=root
  swing1LimitDegrees=45
  swing1Motion=Free
  swing2LimitDegrees=45
  swing2Motion=Free
  twistLimitDegrees=45
  twistMotion=Free
 }
 {bone1Name=jacket_fr_1b_l, bone2Name=jacket_fr_1_l, swing1LimitDegrees=60, swing2LimitDegrees=60, twistLimitDegrees=15}
 {
  bone1Name=spine_01
  bone2Name=pelvis
  swing1LimitDegrees=45
  swing1Motion=Free
  swing2LimitDegrees=45
  swing2Motion=Free
  twistLimitDegrees=45
  twistMotion=Free
 }
 {bone1Name=jacket_bk_4_r, bone2Name=jacket_bk_4_l, swing1LimitDegrees=60, swing2LimitDegrees=60, twistLimitDegrees=15}
 {bone1Name=jacket_fr_1b_r, bone2Name=jacket_fr_1_r, swing1LimitDegrees=60, swing2LimitDegrees=60, twistLimitDegrees=15}
 {bone1Name=jacket_mid_1_l, bone2Name=jacket_mid_base_l, swing1LimitDegrees=30, swing2LimitDegrees=30, twistLimitDegrees=10}
 {bone1Name=jacket_mid_2_l, bone2Name=jacket_mid_1_l, swing1LimitDegrees=30, swing2LimitDegrees=30, twistLimitDegrees=10}
 {bone1Name=jacket_mid_3_l, bone2Name=jacket_mid_2_l, swing1LimitDegrees=60, swing2LimitDegrees=60, twistLimitDegrees=10}
 {bone1Name=jacket_mid_4_l, bone2Name=jacket_mid_3_l, swing1LimitDegrees=40, swing2LimitDegrees=20, twistLimitDegrees=10}
 {bone1Name=jacket_mid_1_r, bone2Name=jacket_mid_base_r, swing1LimitDegrees=30, swing2LimitDegrees=30, twistLimitDegrees=10}
 {bone1Name=jacket_mid_2_r, bone2Name=jacket_mid_1_r, swing1LimitDegrees=30, swing2LimitDegrees=30, twistLimitDegrees=10}
 {bone1Name=jacket_mid_3_r, bone2Name=jacket_mid_2_r, swing1LimitDegrees=60, swing2LimitDegrees=60, twistLimitDegrees=10}
 {bone1Name=jacket_mid_4_r, bone2Name=jacket_mid_3_r, swing1LimitDegrees=40, swing2LimitDegrees=20, twistLimitDegrees=10}
 {
  bone1Name=jacket_bk_1_r
  bone2Name=jacket_bk_1_l
  swing1LimitDegrees=45
  swing1Motion=Free
  swing2LimitDegrees=45
  swing2Motion=Free
  twistLimitDegrees=45
  twistMotion=Free
 }
 {bone1Name=jacket_fr_4_r, bone2Name=jacket_mid_4_r, swing1LimitDegrees=60, swing2LimitDegrees=60, twistLimitDegrees=15}
 {bone1Name=jacket_fr_4_l, bone2Name=jacket_mid_4_l, swing1LimitDegrees=60, swing2LimitDegrees=60, twistLimitDegrees=15}
 {bone1Name=jacket_fr_3_r, bone2Name=jacket_mid_3_r, swing1LimitDegrees=60, swing2LimitDegrees=60, twistLimitDegrees=15}
 {bone1Name=jacket_bk_3_r, bone2Name=jacket_mid_3_r, swing1LimitDegrees=60, swing2LimitDegrees=60, twistLimitDegrees=15}
 {bone1Name=jacket_bk_3_l, bone2Name=jacket_mid_3_l, swing1LimitDegrees=60, swing2LimitDegrees=60, twistLimitDegrees=15}
 {bone1Name=jacket_fr_3_l, bone2Name=jacket_mid_3_l, swing1LimitDegrees=60, swing2LimitDegrees=60, twistLimitDegrees=15}
 {bone1Name=jacket_bk_4_r, bone2Name=jacket_mid_4_r, swing1LimitDegrees=60, swing2LimitDegrees=60, twistLimitDegrees=15}
 {
  bone1Name=jacket_fr_base_r
  bone2Name=pelvis
  swing1LimitDegrees=45
  swing1Motion=Free
  swing2LimitDegrees=45
  swing2Motion=Free
  twistLimitDegrees=45
  twistMotion=Free
 }
 {
  bone1Name=jacket_bk_base_r
  bone2Name=pelvis
  swing1LimitDegrees=45
  swing1Motion=Free
  swing2LimitDegrees=45
  swing2Motion=Free
  twistLimitDegrees=45
  twistMotion=Free
 }
 {
  bone1Name=jacket_bk_base_l
  bone2Name=pelvis
  swing1LimitDegrees=45
  swing1Motion=Free
  swing2LimitDegrees=45
  swing2Motion=Free
  twistLimitDegrees=45
  twistMotion=Free
 }
 {
  bone1Name=jacket_fr_base_l
  bone2Name=pelvis
  swing1LimitDegrees=45
  swing1Motion=Free
  swing2LimitDegrees=45
  swing2Motion=Free
  twistLimitDegrees=45
  twistMotion=Free
 }
 {
  bone1Name=jacket_mid_base_l
  bone2Name=pelvis
  swing1LimitDegrees=45
  swing1Motion=Free
  swing2LimitDegrees=45
  swing2Motion=Free
  twistLimitDegrees=45
  twistMotion=Free
 }
 {
  bone1Name=jacket_mid_base_r
  bone2Name=pelvis
  swing1LimitDegrees=45
  swing1Motion=Free
  swing2LimitDegrees=45
  swing2Motion=Free
  twistLimitDegrees=45
  twistMotion=Free
 }
]
