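# Post-process style AnimBlueprint for the Ruckus male body that copies the parent mesh pose and animates its jacket.
# Rigid body simulations for front end, skydiving (dynamic gravity) and normal play feed a Control Rig.
# With rigid bodies disabled, additive bone offsets and constraints driven by thigh and knee angles correct the jacket.
# The event graph computes those inputs, suppresses simulation and shrinks the jacket for certain vehicle drivers, and toggles a material hide parameter.
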AnimBlueprint /BRCosmetics/Characters/Player/Male/Medium/Bodies/M_MED_Ruckus/Meshes/M_MED_Ruckus_AnimBP
target_skeleton=&/M_MED_Ruckus_Skeleton
graphs [
 AnimGraph {nodes [
   n1 = Get knee_bend_r() -> (knee_bend_r:double)
   "Constraint - jacket_mid_base_r (T)"(Alpha=@n1, ConstraintWeights_0=1) {
    alphaBoolBlend {blendOption=Linear}
    boneToModify {boneName=jacket_mid_base_r}
    constraintSetup [{transformType=Translation, targetBone {boneName=pelvis}}]
    constraintWeights [1]
   }
   n2 = Get sliding() -> (sliding:double)
   n3 = "Transform (Modify) Bone - Bone: jacket_bk_1_r"(Alpha=@n2, Translation="0,0,0", Rotation="0,-75,0", Scale="1,1,1") {
    rotationMode=BMM_Additive
    rotationSpace=BCS_BoneSpace
    translationMode=BMM_Additive
    translationSpace=BCS_BoneSpace
    alphaBoolBlend {blendOption=Linear}
    boneToModify {boneName=jacket_bk_1_r}
    rotation {yaw=-75}
   }
   n4 = "Transform (Modify) Bone - Bone: jacket_bk_1_l"(ComponentPose=@n3, Alpha=@n2, Translation="0,0,0", Rotation="0,-75,0", Scale="1,1,1") {
    rotationMode=BMM_Additive
    rotationSpace=BCS_BoneSpace
    translationMode=BMM_Additive
    translationSpace=BCS_BoneSpace
    alphaBoolBlend {blendOption=Linear}
    boneToModify {boneName=jacket_bk_1_l}
    rotation {yaw=-75}
   }
   n5 = "Transform (Modify) Bone - Bone: jacket_mid_1_l"(ComponentPose=@n4, Alpha=@n2, Translation="0,0,0", Rotation="-75,-20,0", Scale="1,1,1") {
    rotationMode=BMM_Additive
    rotationSpace=BCS_BoneSpace
    translationMode=BMM_Additive
    translationSpace=BCS_BoneSpace
    alphaBoolBlend {blendOption=Linear}
    boneToModify {boneName=jacket_mid_1_l}
    rotation {pitch=-75, yaw=-20}
   }
   n6 = "Transform (Modify) Bone - Bone: jacket_mid_1_r"(ComponentPose=@n5, Alpha=@n2, Translation="0,0,0", Rotation="75,-20,0", Scale="1,1,1") {
    rotationMode=BMM_Additive
    rotationSpace=BCS_BoneSpace
    translationMode=BMM_Additive
    translationSpace=BCS_BoneSpace
    alphaBoolBlend {blendOption=Linear}
    boneToModify {boneName=jacket_mid_1_r}
    rotation {pitch=75, yaw=-20}
   }
   n7 = "Transform (Modify) Bone - Bone: jacket_fr_1b_l"(ComponentPose=@n6, Alpha=@n2, Translation="0,0,0", Rotation="-60,0,0", Scale="1,1,1") {
    rotationMode=BMM_Additive
    rotationSpace=BCS_BoneSpace
    translationMode=BMM_Additive
    translationSpace=BCS_BoneSpace
    alphaBoolBlend {blendOption=Linear}
    boneToModify {boneName=jacket_fr_1b_l}
    rotation {pitch=-60}
   }
   "Transform (Modify) Bone - Bone: jacket_fr_1b_r"(ComponentPose=@n7, Alpha=@n2, Translation="0,0,0", Rotation="60,0,0", Scale="1,1,1") {
    rotationMode=BMM_Additive
    rotationSpace=BCS_BoneSpace
    translationMode=BMM_Additive
    translationSpace=BCS_BoneSpace
    alphaBoolBlend {blendOption=Linear}
    boneToModify {boneName=jacket_fr_1b_r}
    rotation {pitch=60}
   }
   "Transform (Modify) Bone - Bone: jacket_bk_1_r"(Alpha=1, Translation="0,0,0", Rotation="0,-85,-45", Scale="1,1,1") {
    rotationMode=BMM_Additive
    rotationSpace=BCS_BoneSpace
    translationMode=BMM_Additive
    translationSpace=BCS_BoneSpace
    alphaBoolBlend {blendInTime=0.5, blendOption=Linear, blendOutTime=0.1}
    alphaScaleBiasClamp {bInterpResult=true, bMapRange=true}
    boneToModify {boneName=jacket_bk_1_r}
    rotation {roll=-45, yaw=-85}
   }
   n8 = "Transform (Modify) Bone - Bone: jacket_mid_base_l"(Alpha=1, Translation="0.000,-2.000,-2", Rotation="0,0,0", Scale="1,1,1") {
    translationMode=BMM_Additive
    translationSpace=BCS_ParentBoneSpace
    alphaBoolBlend {blendOption=Linear}
    boneToModify {boneName=jacket_mid_base_l}
    translation {y=-2, z=-2}
   }
   n9 = "Transform (Modify) Bone - Bone: jacket_bk_base_l"(ComponentPose=@n8, Alpha=1, Translation="0,0,0", Rotation="0,0,0", Scale="1,1,1") {
    translationMode=BMM_Additive
    translationSpace=BCS_ParentBoneSpace
    alphaBoolBlend {blendOption=Linear}
    boneToModify {boneName=jacket_bk_base_l}
   }
   n10 = "Transform (Modify) Bone - Bone: jacket_bk_base_r"(ComponentPose=@n9, Alpha=1, Translation="0,0,0", Rotation="0,0,0", Scale="1,1,1") {
    translationMode=BMM_Additive
    translationSpace=BCS_ParentBoneSpace
    alphaBoolBlend {blendOption=Linear}
    boneToModify {boneName=jacket_bk_base_r}
   }
   "Transform (Modify) Bone - Bone: jacket_mid_base_r"(ComponentPose=@n10, Alpha=1, Translation="0,-2,2", Rotation="0,0,0", Scale="1,1,1") {
    translationMode=BMM_Additive
    translationSpace=BCS_ParentBoneSpace
    alphaBoolBlend {blendOption=Linear}
    boneToModify {boneName=jacket_mid_base_r}
    translation {y=-2, z=2}
   }
   n11 = "Transform (Modify) Bone - Bone: jacket_fr_1b_r"(ComponentPose=Local To Component(LocalPose=Use cached pose 'Cache2'()), Alpha=1, Translation="0,-2,5", Rotation="5,5,0", Scale="1,1,1") {
    rotationMode=BMM_Additive
    rotationSpace=BCS_BoneSpace
    translationMode=BMM_Additive
    translationSpace=BCS_BoneSpace
    alphaBoolBlend {blendOption=Linear}
    boneToModify {boneName=jacket_fr_1b_r}
    rotation {pitch=5, yaw=5}
    translation {y=-2, z=5}
   }
   n12 = "Transform (Modify) Bone - Bone: jacket_mid_1_r"(ComponentPose=@n11, Alpha=1, Translation="0,0,0", Rotation="5,0,0", Scale="1,1,1") {
    rotationMode=BMM_Additive
    rotationSpace=BCS_BoneSpace
    translationMode=BMM_Additive
    translationSpace=BCS_BoneSpace
    alphaBoolBlend {blendOption=Linear}
    boneToModify {boneName=jacket_mid_1_r}
    rotation {pitch=5}
   }
   n13 = "Transform (Modify) Bone - Bone: jacket_bk_1_r"(ComponentPose=@n12, Alpha=1, Translation="0,0,0", Rotation="0,-5,0", Scale="1,1,1") {
    rotationMode=BMM_Additive
    rotationSpace=BCS_BoneSpace
    translationMode=BMM_Additive
    translationSpace=BCS_BoneSpace
    alphaBoolBlend {blendOption=Linear}
    boneToModify {boneName=jacket_bk_1_r}
    rotation {yaw=-5}
   }
   n14 = "Transform (Modify) Bone - Bone: jacket_fr_1b_l"(ComponentPose=@n13, Alpha=1, Translation="0,0,-5", Rotation="-8,-5,0", Scale="1,1,1") {
    rotationMode=BMM_Additive
    rotationSpace=BCS_BoneSpace
    translationMode=BMM_Additive
    translationSpace=BCS_BoneSpace
    alphaBoolBlend {blendOption=Linear}
    boneToModify {boneName=jacket_fr_1b_l}
    rotation {pitch=-8, yaw=-5}
    translation {z=-5}
   }
   n15 = "Transform (Modify) Bone - Bone: jacket_mid_1_l"(ComponentPose=@n14, Alpha=1, Translation="0,0,0", Rotation="-5,0,0", Scale="1,1,1") {
    rotationMode=BMM_Additive
    rotationSpace=BCS_BoneSpace
    translationMode=BMM_Additive
    translationSpace=BCS_BoneSpace
    alphaBoolBlend {blendOption=Linear}
    boneToModify {boneName=jacket_mid_1_l}
    rotation {pitch=-5}
   }
   n16 = "Transform (Modify) Bone - Bone: jacket_bk_1_l"(ComponentPose=@n15, Alpha=1, Translation="0,0,0", Rotation="0,-5,0", Scale="1,1,1") {
    rotationMode=BMM_Additive
    rotationSpace=BCS_BoneSpace
    translationMode=BMM_Additive
    translationSpace=BCS_BoneSpace
    alphaBoolBlend {blendOption=Linear}
    boneToModify {boneName=jacket_bk_1_l}
    rotation {yaw=-5}
   }
   n17 = Get thigh_rot_up_r() -> (thigh_rot_up_r:double)
   n18 = "Transform (Modify) Bone - Bone: jacket_mid_1_r"(ComponentPose=@n16, Alpha=@n17, Translation="0,0,0", Rotation="5,0,0", Scale="1,1,1") {
    rotationMode=BMM_Additive
    rotationSpace=BCS_BoneSpace
    translationMode=BMM_Additive
    translationSpace=BCS_BoneSpace
    alphaBoolBlend {blendOption=Linear}
    boneToModify {boneName=jacket_mid_1_r}
    rotation {pitch=5}
   }
   n19 = "Transform (Modify) Bone - Bone: jacket_fr_1b_r"(ComponentPose=@n18, Alpha=@n17, Translation="0,0,5", Rotation="5,5,0", Scale="1,1,1") {
    rotationMode=BMM_Additive
    rotationSpace=BCS_BoneSpace
    scaleMode=BMM_Additive
    scaleSpace=BCS_BoneSpace
    translationMode=BMM_Additive
    translationSpace=BCS_BoneSpace
    alphaBoolBlend {blendOption=Linear}
    boneToModify {boneName=jacket_fr_1b_r}
    rotation {pitch=5, yaw=5}
    translation {z=5}
   }
   n20 = Get thigh_rot_up_l() -> (thigh_rot_up_l:double)
   n21 = "Transform (Modify) Bone - Bone: jacket_mid_1_l"(ComponentPose=@n19, Alpha=@n20, Translation="0,0,0", Rotation="5,0,0", Scale="1,1,1") {
    rotationMode=BMM_Additive
    rotationSpace=BCS_BoneSpace
    translationMode=BMM_Additive
    translationSpace=BCS_BoneSpace
    alphaBoolBlend {blendOption=Linear}
    boneToModify {boneName=jacket_mid_1_l}
    rotation {pitch=5}
   }
   n22 = "Transform (Modify) Bone - Bone: jacket_fr_1b_l"(ComponentPose=@n21, Alpha=@n20, Translation="0,0,-5", Rotation="0,0,0", Scale="1,1,1") {
    rotationMode=BMM_Additive
    rotationSpace=BCS_BoneSpace
    scaleMode=BMM_Additive
    scaleSpace=BCS_BoneSpace
    translationMode=BMM_Additive
    translationSpace=BCS_BoneSpace
    alphaBoolBlend {blendOption=Linear}
    boneToModify {boneName=jacket_fr_1b_l}
    translation {z=-5}
   }
   n23 = "Constraint - jacket_bk_base_l (T)"(ComponentPose=@n22, Alpha=@n1, ConstraintWeights_0=1) {
    alphaBoolBlend {blendOption=Linear}
    boneToModify {boneName=jacket_bk_base_l}
    constraintSetup [{transformType=Translation, targetBone {boneName=pelvis}}]
    constraintWeights [1]
   }
   n24 = "Constraint - jacket_bk_base_r (T)"(ComponentPose=@n23, Alpha=@n1, ConstraintWeights_0=1) {
    alphaBoolBlend {blendOption=Linear}
    boneToModify {boneName=jacket_bk_base_r}
    constraintSetup [{transformType=Translation, targetBone {boneName=pelvis}}]
    constraintWeights [1]
   }
   n25 = "Constraint - jacket_mid_base_l (T)"(ComponentPose=@n24, Alpha=@n1, ConstraintWeights_0=1) {
    alphaBoolBlend {blendOption=Linear}
    boneToModify {boneName=jacket_mid_base_l}
    constraintSetup [{transformType=Translation, targetBone {boneName=pelvis}}]
    constraintWeights [1]
   }
   n26 = "Transform (Modify) Bone - Bone: jacket_fr_1b_r"(ComponentPose=@n25, Alpha=@n1, Translation="0,0,0", Rotation="5,5,-10", Scale="1,1,1") {
    rotationMode=BMM_Additive
    rotationSpace=BCS_BoneSpace
    translationMode=BMM_Additive
    translationSpace=BCS_BoneSpace
    alphaBoolBlend {blendOption=Linear}
    boneToModify {boneName=jacket_fr_1b_r}
    rotation=<rot 5 5 -10>
   }
   n27 = "Transform (Modify) Bone - Bone: jacket_fr_2_r"(ComponentPose=@n26, Alpha=@n1, Translation="0,0,0", Rotation="25,0,0", Scale="1,1,1") {
    rotationMode=BMM_Additive
    rotationSpace=BCS_BoneSpace
    translationMode=BMM_Additive
    translationSpace=BCS_BoneSpace
    alphaBoolBlend {blendOption=Linear}
    boneToModify {boneName=jacket_fr_2_r}
    rotation {pitch=25}
   }
   n28 = "Transform (Modify) Bone - Bone: jacket_mid_2_r"(ComponentPose=@n27, Alpha=@n1, Translation="0,0,0", Rotation="25,-5,0", Scale="1,1,1") {
    rotationMode=BMM_Additive
    rotationSpace=BCS_BoneSpace
    translationMode=BMM_Additive
    translationSpace=BCS_BoneSpace
    alphaBoolBlend {blendOption=Linear}
    boneToModify {boneName=jacket_mid_2_r}
    rotation {pitch=25, yaw=-5}
   }
   n29 = "Transform (Modify) Bone - Bone: jacket_mid_3_r"(ComponentPose=@n28, Alpha=@n1, Translation="0,0,0", Rotation="10,0,0", Scale="1,1,1") {
    rotationMode=BMM_Additive
    rotationSpace=BCS_BoneSpace
    translationMode=BMM_Additive
    translationSpace=BCS_BoneSpace
    alphaBoolBlend {blendOption=Linear}
    boneToModify {boneName=jacket_mid_3_r}
    rotation {pitch=10}
   }
   n30 = "Transform (Modify) Bone - Bone: jacket_bk_2_r"(ComponentPose=@n29, Alpha=@n1, Translation="0,0,0", Rotation="0,-10,0", Scale="1,1,1") {
    rotationMode=BMM_Additive
    rotationSpace=BCS_BoneSpace
    translationMode=BMM_Additive
    translationSpace=BCS_BoneSpace
    alphaBoolBlend {blendOption=Linear}
    boneToModify {boneName=jacket_bk_2_r}
    rotation {yaw=-10}
   }
   n31 = "Transform (Modify) Bone - Bone: jacket_bk_3_r"(ComponentPose=@n30, Alpha=@n1, Translation="0,0,0", Rotation="0,0,0", Scale="1,1,1") {
    rotationMode=BMM_Additive
    rotationSpace=BCS_BoneSpace
    translationMode=BMM_Additive
    translationSpace=BCS_BoneSpace
    alphaBoolBlend {blendOption=Linear}
    boneToModify {boneName=jacket_bk_3_r}
   }
   n32 = "Transform (Modify) Bone - Bone: jacket_bk_1_r"!(ComponentPose=@n31, bAlphaBoolEnabled:bool=True, Translation="0,0,0", Rotation="0,-10,0", Scale="1,1,1") -> (Pose) {
    alphaInputType=Bool
    rotationMode=BMM_Additive
    rotationSpace=BCS_BoneSpace
    translationMode=BMM_Additive
    translationSpace=BCS_BoneSpace
    alphaBoolBlend {blendInTime=0.1, blendOption=Linear, blendOutTime=0.1}
    boneToModify {boneName=jacket_bk_1_r}
    rotation {yaw=-10}
   }
   n33 = "Transform (Modify) Bone - Bone: jacket_bk_1_l"!(ComponentPose=@n32, bAlphaBoolEnabled:bool=True, Translation="0,0,0", Rotation="0,-10,0", Scale="1,1,1") -> (Pose) {
    alphaInputType=Bool
    rotationMode=BMM_Additive
    rotationSpace=BCS_BoneSpace
    translationMode=BMM_Additive
    translationSpace=BCS_BoneSpace
    alphaBoolBlend {blendInTime=0.1, blendOption=Linear, blendOutTime=0.1}
    boneToModify {boneName=jacket_bk_1_l}
    rotation {yaw=-10}
   }
   n34 = "Transform (Modify) Bone - Bone: jacket_bk_2_r"!(ComponentPose=@n33, bAlphaBoolEnabled:bool=True, Translation="0,0,0", Rotation="0,-10,0", Scale="1,1,1") -> (Pose) {
    alphaInputType=Bool
    rotationMode=BMM_Additive
    rotationSpace=BCS_BoneSpace
    translationMode=BMM_Additive
    translationSpace=BCS_BoneSpace
    alphaBoolBlend {blendInTime=0.1, blendOption=Linear, blendOutTime=0.1}
    boneToModify {boneName=jacket_bk_2_r}
    rotation {yaw=-10}
   }
   n35 = "Transform (Modify) Bone - Bone: jacket_mid_2_r"!(ComponentPose=@n34, bAlphaBoolEnabled:bool=True, Translation="0,0,0", Rotation="0,-10,0", Scale="1,1,1") -> (Pose) {
    alphaInputType=Bool
    rotationMode=BMM_Additive
    rotationSpace=BCS_BoneSpace
    translationMode=BMM_Additive
    translationSpace=BCS_BoneSpace
    alphaBoolBlend {blendInTime=0.1, blendOption=Linear, blendOutTime=0.1}
    boneToModify {boneName=jacket_mid_2_r}
    rotation {yaw=-10}
   }
   n36 = "Transform (Modify) Bone - Bone: jacket_bk_2_l"!(ComponentPose=@n35, bAlphaBoolEnabled:bool=True, Translation="0,0,0", Rotation="0,-10,0", Scale="1,1,1") -> (Pose) {
    alphaInputType=Bool
    rotationMode=BMM_Additive
    rotationSpace=BCS_BoneSpace
    translationMode=BMM_Additive
    translationSpace=BCS_BoneSpace
    alphaBoolBlend {blendInTime=0.1, blendOption=Linear, blendOutTime=0.1}
    boneToModify {boneName=jacket_bk_2_l}
    rotation {yaw=-10}
   }
   n37 = "Transform (Modify) Bone - Bone: jacket_mid_2_l"!(ComponentPose=@n36, bAlphaBoolEnabled:bool=True, Translation="0,0,0", Rotation="0,-10,0", Scale="1,1,1") -> (Pose) {
    alphaInputType=Bool
    rotationMode=BMM_Additive
    rotationSpace=BCS_BoneSpace
    translationMode=BMM_Additive
    translationSpace=BCS_BoneSpace
    alphaBoolBlend {blendInTime=0.1, blendOption=Linear, blendOutTime=0.1}
    boneToModify {boneName=jacket_mid_2_l}
    rotation {yaw=-10}
   }
   n38 = Component To Local(ComponentPose=@n37)
   n39 = "Constraint - jacket (P)"(ComponentPose=Local To Component(LocalPose=Use cached pose 'Cache2'()), Alpha=1, ConstraintWeights_0=1) {
    alphaBoolBlend {blendOption=Linear}
    boneToModify {boneName=jacket}
    constraintSetup [{targetBone {boneName=spine_01}}]
    constraintWeights [1]
   }
   n40 = "Transform (Modify) Bone - Bone: jacket_bk_1_r"!(ComponentPose=@n39, Alpha:float=1, Translation="0,0,0", Rotation="0,-90,0", Scale="1,1,1") -> (Pose) {
    rotationMode=BMM_Additive
    rotationSpace=BCS_BoneSpace
    translationMode=BMM_Additive
    translationSpace=BCS_BoneSpace
    alphaBoolBlend {blendInTime=0.5, blendOption=Linear, blendOutTime=0.1}
    alphaScaleBiasClamp {bInterpResult=true, bMapRange=true}
    boneToModify {boneName=jacket_bk_1_r}
    rotation {yaw=-90}
   }
   n41 = "Transform (Modify) Bone - Bone: jacket_bk_1_l"!(ComponentPose=@n40, Alpha:float=1, Translation="0,0,0", Rotation="0,-85,30", Scale="1,1,1") -> (Pose) {
    rotationMode=BMM_Additive
    rotationSpace=BCS_BoneSpace
    translationMode=BMM_Additive
    translationSpace=BCS_BoneSpace
    alphaBoolBlend {blendInTime=0.5, blendOption=Linear, blendOutTime=0.1}
    alphaScaleBiasClamp {bInterpResult=true, bMapRange=true}
    boneToModify {boneName=jacket_bk_1_l}
    rotation {roll=30, yaw=-85}
   }
   n42 = "Transform (Modify) Bone - Bone: jacket_mid_1_l"(ComponentPose=@n41, Alpha=1, Translation="0,0,0", Rotation="-80,-20,20", Scale="1,1,1") {
    rotationMode=BMM_Additive
    rotationSpace=BCS_BoneSpace
    translationMode=BMM_Additive
    translationSpace=BCS_BoneSpace
    alphaBoolBlend {blendInTime=0.5, blendOption=Linear, blendOutTime=0.1}
    alphaScaleBiasClamp {bInterpResult=true, bMapRange=true}
    boneToModify {boneName=jacket_mid_1_l}
    rotation=<rot -80 -20 20>
   }
   n43 = "Transform (Modify) Bone - Bone: jacket_mid_1_r"(ComponentPose=@n42, Alpha=1, Translation="0,0,0", Rotation="60,-70,0", Scale="1,1,1") {
    rotationMode=BMM_Additive
    rotationSpace=BCS_BoneSpace
    translationMode=BMM_Additive
    translationSpace=BCS_BoneSpace
    alphaBoolBlend {blendInTime=0.5, blendOption=Linear, blendOutTime=0.1}
    alphaScaleBiasClamp {bInterpResult=true, bMapRange=true}
    boneToModify {boneName=jacket_mid_1_r}
    rotation {pitch=60, yaw=-70}
   }
   n44 = "Transform (Modify) Bone - Bone: jacket_mid_2_r"!(ComponentPose=@n43, Alpha:float=1, Translation="0,0,0", Rotation="0,-10,0", Scale="1,1,1") -> (Pose) {
    rotationMode=BMM_Additive
    rotationSpace=BCS_BoneSpace
    translationMode=BMM_Additive
    translationSpace=BCS_BoneSpace
    alphaBoolBlend {blendInTime=0.5, blendOption=Linear, blendOutTime=0.1}
    alphaScaleBiasClamp {bInterpResult=true, bMapRange=true}
    boneToModify {boneName=jacket_mid_2_r}
    rotation {yaw=-10}
   }
   n45 = "Transform (Modify) Bone - Bone: jacket_fr_1b_l"(ComponentPose=@n44, Alpha=1, Translation="0,0,0", Rotation="-60,0,0", Scale="1,1,1") {
    rotationMode=BMM_Additive
    rotationSpace=BCS_BoneSpace
    translationMode=BMM_Additive
    translationSpace=BCS_BoneSpace
    alphaBoolBlend {blendInTime=0.5, blendOption=Linear, blendOutTime=0.1}
    alphaScaleBiasClamp {bInterpResult=true, bMapRange=true}
    boneToModify {boneName=jacket_fr_1b_l}
    rotation {pitch=-60}
   }
   n46 = "Transform (Modify) Bone - Bone: jacket_fr_1b_r"(ComponentPose=@n45, Alpha=1, Translation="0,0,0", Rotation="60,0,0", Scale="1,1,1") {
    rotationMode=BMM_Additive
    rotationSpace=BCS_BoneSpace
    translationMode=BMM_Additive
    translationSpace=BCS_BoneSpace
    alphaBoolBlend {blendInTime=0.5, blendOption=Linear, blendOutTime=0.1}
    alphaScaleBiasClamp {bInterpResult=true, bMapRange=true}
    boneToModify {boneName=jacket_fr_1b_r}
    rotation {pitch=60}
   }
   n47 = Component To Local(ComponentPose=@n46)
   Output Pose(Result=Blend Poses by bool(
     BlendPose_0=Use cached pose 'Cache2'()
     BlendPose_1=Blend Poses by bool(BlendPose_0=@n47, BlendPose_1=@n38, BlendTime_1=0.5) {blendType=Linear}
     bActiveValue=Get bRigidBodyEnabled()
    ) {blendType=Linear}) {layerGroup=None}
   n48 = "Transform (Modify) Bone - Bone: dyn_slope_foot_c"!(
    ComponentPose=Local To Component(LocalPose=Copy Pose From Mesh() {bUseAttachedParent=true})
    bAlphaBoolEnabled:bool=NOT Boolean(A=Get bCanPlayCustomAnimations() -> (bCanPlayCustomAnimations:bool))
    Translation=Get slope_foot_trans() -> (slope_foot_trans:Vector)
    Rotation=Get slope_foot_rot() -> (slope_foot_rot:Rotator)
    Scale="1,1,1"
   ) -> (Pose) {
    alphaInputType=Bool
    rotationMode=BMM_Replace
    rotationSpace=BCS_WorldSpace
    translationMode=BMM_Replace
    translationSpace=BCS_WorldSpace
    alphaBoolBlend {blendOption=Linear}
    boneToModify {boneName=dyn_slope_foot_c}
   }
   Save cached pose 'RBAN_input_pose'(Pose=Component To Local(ComponentPose=@n48)) {cacheName=RBAN_input_pose}
   n49 = RigidBody!(
    ComponentPose=Local To Component(LocalPose=Use cached pose 'RBAN_input_pose'())
    Alpha
    OverrideWorldGravity:Vector=Get GravityOverride_b() -> (GravityOverride_b:Vector)
    ExternalForce
   ) -> (Pose) {
    bClampLinearTranslationLimitToRefPose=true
    bOverrideWorldGravity=true
    overridePhysicsAsset=&/M_MED_Ruckus_Jacket_Physics
    simulationSpace=BaseBoneSpace
    alphaBoolBlend {blendOption=Linear}
    baseBoneRef {boneName=pelvis}
    componentAppliedLinearAccClamp=<1000 1000 1000>
    componentLinearAccScale=<0.15 0.15 0.3>
    componentLinearVelScale=<0.5 0.5 0.3>
    overrideWorldGravity {z=-200}
    simSpaceSettings {maxAngularAcceleration=1000, maxAngularVelocity=1000}
   }
   n50 = Component To Local(ComponentPose=@n49)
   n51 = RigidBody!(
    ComponentPose=Local To Component(LocalPose=Use cached pose 'RBAN_input_pose'())
    Alpha=Get in vehicle rigid alpha() -> (in vehicle rigid alpha:double)
    ExternalForce
   ) -> (Pose) {
    bClampLinearTranslationLimitToRefPose=true
    bOverrideWorldGravity=true
    overridePhysicsAsset=&M_MED_Ruckus_Jacket_Physics
    simulationSpace=BaseBoneSpace
    alphaBoolBlend {blendOption=Linear}
    baseBoneRef {boneName=root}
    componentAppliedLinearAccClamp=<1000 1000 1000>
    componentLinearAccScale=<0.5 0.5 0.4>
    componentLinearVelScale=<0.5 0.5 0.4>
    overrideWorldGravity {z=-920}
   }
   n52 = Component To Local(ComponentPose=@n51)
   n53 = Get bIsSkyDiving()
   n54 = Get bRigidBodyEnabled()
   n55 = "Transform (Modify) Bone - Bone: jacket_fr_base_r"!(
    ComponentPose=Local To Component(LocalPose=Use cached pose 'RBAN_input_pose'())
    bAlphaBoolEnabled:bool=@n54
    Translation="0,0,0"
    Rotation="10,-5,0"
    Scale="1,1,1"
   ) -> (Pose) {
    alphaInputType=Bool
    rotationMode=BMM_Additive
    rotationSpace=BCS_BoneSpace
    translationMode=BMM_Additive
    translationSpace=BCS_BoneSpace
    alphaBoolBlend {blendInTime=0.1, blendOption=Linear, blendOutTime=0.1}
    boneToModify {boneName=jacket_fr_base_r}
    rotation {pitch=10, yaw=-5}
   }
   n56 = "Transform (Modify) Bone - Bone: jacket_fr_base_l"!(ComponentPose=@n55, bAlphaBoolEnabled:bool=@n54, Translation="0,0,0", Rotation="-10,0,0", Scale="1,1,1") -> (Pose) {
    alphaInputType=Bool
    rotationMode=BMM_Additive
    rotationSpace=BCS_BoneSpace
    translationMode=BMM_Additive
    translationSpace=BCS_BoneSpace
    alphaBoolBlend {blendInTime=0.1, blendOption=Linear, blendOutTime=0.1}
    boneToModify {boneName=jacket_fr_base_l}
    rotation {pitch=-10}
   }
   n57 = RigidBody(ComponentPose=@n56) {
    bClampLinearTranslationLimitToRefPose=true
    bOverrideWorldGravity=true
    overridePhysicsAsset=&/M_MED_Ruckus_Jacket_frontend_Physics
    simulationSpace=BaseBoneSpace
    alphaBoolBlend {blendOption=Linear}
    baseBoneRef {boneName=root}
    componentAppliedLinearAccClamp=<1000 1000 1000>
    componentLinearAccScale=<0.5 0.5 0.4>
    componentLinearVelScale=<0.5 0.5 0.4>
    overrideWorldGravity {z=-920}
   }
   n58 = Control Rig(Source=Blend Poses by bool(
     BlendPose_0=Component To Local(ComponentPose=@n57)
     BlendPose_1=Blend Poses by bool(BlendPose_0=@n50, BlendPose_1=@n52, bActiveValue=@n53) {blendType=Linear}
     bActiveValue=Get bIsInFrontEnd()
    ) {blendType=Linear}) {
    alphaBoolBlend {blendOption=Linear}
    controlRigAssetReference {blueprintRigClass=&/M_MED_Ruckus_CtrlRig.M_MED_Ruckus_CtrlRig_C}
   }
   n59 = "Transform (Modify) Bone - Bone: jacket_mid_3_l"(ComponentPose=Local To Component(LocalPose=@n58), Alpha=1, Translation="20,0,0", Rotation="0,0,0", Scale="1,1,1") {
    translationMode=BMM_Replace
    translationSpace=BCS_ParentBoneSpace
    alphaBoolBlend {blendOption=Linear}
    boneToModify {boneName=jacket_mid_3_l}
    translation {x=20}
   }
   n60 = "Transform (Modify) Bone - Bone: jacket_bk_3_l"(ComponentPose=@n59, Alpha=1, Translation="20,0,0", Rotation="0,0,0", Scale="1,1,1") {
    translationMode=BMM_Replace
    translationSpace=BCS_ParentBoneSpace
    alphaBoolBlend {blendOption=Linear}
    boneToModify {boneName=jacket_bk_3_l}
    translation {x=20}
   }
   n61 = "Transform (Modify) Bone - Bone: jacket_bk_3_r"(ComponentPose=@n60, Alpha=1, Translation="20,0,0", Rotation="0,0,0", Scale="1,1,1") {
    translationMode=BMM_Replace
    translationSpace=BCS_ParentBoneSpace
    alphaBoolBlend {blendOption=Linear}
    boneToModify {boneName=jacket_bk_3_r}
    translation {x=20}
   }
   n62 = "Transform (Modify) Bone - Bone: jacket_mid_3_r"(ComponentPose=@n61, Alpha=1, Translation="20,0,0", Rotation="0,0,0", Scale="1,1,1") {
    translationMode=BMM_Replace
    translationSpace=BCS_ParentBoneSpace
    alphaBoolBlend {blendOption=Linear}
    boneToModify {boneName=jacket_mid_3_r}
    translation {x=20}
   }
   n63 = "Transform (Modify) Bone - Bone: jacket_mid_base_l"(ComponentPose=@n62, Alpha=1, Translation="0.000,-2.000,-2", Rotation="0,0,0", Scale="1,1,1") {
    translationMode=BMM_Additive
    translationSpace=BCS_ParentBoneSpace
    alphaBoolBlend {blendOption=Linear}
    boneToModify {boneName=jacket_mid_base_l}
    translation {y=-2, z=-2}
   }
   n64 = "Transform (Modify) Bone - Bone: jacket_bk_base_l"(ComponentPose=@n63, Alpha=1, Translation="0,0,0", Rotation="0,0,0", Scale="1,1,1") {
    translationMode=BMM_Additive
    translationSpace=BCS_ParentBoneSpace
    alphaBoolBlend {blendOption=Linear}
    boneToModify {boneName=jacket_bk_base_l}
   }
   n65 = "Transform (Modify) Bone - Bone: jacket_bk_base_r"(ComponentPose=@n64, Alpha=1, Translation="0,0,0", Rotation="0,0,0", Scale="1,1,1") {
    translationMode=BMM_Additive
    translationSpace=BCS_ParentBoneSpace
    alphaBoolBlend {blendOption=Linear}
    boneToModify {boneName=jacket_bk_base_r}
   }
   n66 = "Transform (Modify) Bone - Bone: jacket_mid_base_r"(ComponentPose=@n65, Alpha=1, Translation="0,-2,2", Rotation="0,0,0", Scale="1,1,1") {
    translationMode=BMM_Additive
    translationSpace=BCS_ParentBoneSpace
    alphaBoolBlend {blendOption=Linear}
    boneToModify {boneName=jacket_mid_base_r}
    translation {y=-2, z=2}
   }
   n67 = "Transform (Modify) Bone - Bone: jacket"(
    ComponentPose=@n66
    Alpha=Get in vehicle scale alpha() -> (in vehicle scale alpha:double)
    Translation="0,0,0"
    Rotation="0,0,0"
    Scale="0.3,1,1"
   ) {scaleMode=BMM_Replace, scaleSpace=BCS_BoneSpace, alphaBoolBlend {blendOption=Linear}, boneToModify {boneName=jacket}, scale {x=0.3}}
   Save cached pose 'Cache2'(Pose=Component To Local(ComponentPose=@n67)) {cacheName=Cache2}
   Get sliding()
   Get sprinting() -> (sprinting:double)
  ]}
 EventGraph {nodes [
   Delay(Duration=0)
   > ResetDynamics(InTeleportType=TeleportPhysics)
   > ResetDynamics(InTeleportType=ResetPhysics)
   n1 = Event BlueprintUpdateAnimation()
   > n2 = Sequence!(execute) -> (then_0, then_1, then_2:Exec, then_3:Exec, then_4:Exec, then_5:Exec, then_6:Exec)
   n3 = Get PlayerPawn() -> (PlayerPawn:Fort Player Pawn Athena Object Reference)
   > n4 = Is Valid(InputObject=@n3)
   n5 = GetSkeletalMeshForPartType(self=@n3, PartType=Backpack)
   Set Ruckus_pack(
    execute=[@n4.Is Valid, @n48.Completed]
    Ruckus_pack:bool=Contains(SearchIn="To String (Object)"!(InObj:Object Reference=@n5) -> (ReturnValue:String), Substring=Ruckus_Pack)
   ) -> (Output_Get:bool)
   n6 = Break Fort Anim Input Slope Info!(FortAnimInput_SlopeInfo:"Fort Anim Input Slope Info Structure (by ref)"=Get SlopeInfoInput() -> (SlopeInfoInput:Fort Anim Input Slope Info Structure)) -> (RootSlopeTranslation:Vector=~, LeftFootSlopeTranslation:Vector=~, RightFootSlopeTranslation:Vector=~, SlopeRotation:Rotator=~) {is=title K2Node_BreakStruct}
   n7 = BreakVector(InVec=@n6)
   n8 = BreakVector(InVec=VInterpTo!(Current:Vector=Get slope_foot_trans(), Target:Vector=@n6, DeltaTime:float=@n1.DeltaTimeX, InterpSpeed:float=4) -> (ReturnValue:Vector))
   Set slope_foot_trans(
    execute=@n2.then_1
    slope_foot_trans:Vector=MakeVector(
     X=@n7
     Y=@n7.Y
     Z="Min (Float)"(A:double=@n8.Z, B:double=@n7.Z) -> (ReturnValue:double) {is=title K2Node_CommutativeAssociativeBinaryOperator}
    )
   ) -> (Output_Get:Vector)
   > Set slope_foot_rot(slope_foot_rot:Rotator=RInterpTo!(Current:Rotator=Get slope_foot_rot(), Target:Rotator=@n6.SlopeRotation, DeltaTime:float=@n1.DeltaTimeX, InterpSpeed:float=2) -> (ReturnValue:Rotator)) -> (Output_Get:Rotator)
   n9 = GetOwnerSkeletalMeshComponent()
   n10 = BreakRotator(InRot=BreakTransform(InTransform=GetDeltaTransformFromRefPose(self=@n9, BoneName=thigh_r, BaseName=pelvis)).Rotation)
   Set thigh_rot_up_r(execute=@n2.then_6, thigh_rot_up_r:double=MapRangeClamped(Value=@n10.Yaw, InRangeA=-15, InRangeB=-100)) -> (Output_Get:double)
   > Set thigh_rot_inr_r(thigh_rot_inr_r:double=MapRangeClamped(Value=@n10.Pitch, InRangeA=0, InRangeB=-10, OutRangeA=1, OutRangeB=0)) -> (Output_Get:double)
   > Set thigh_rot_otr_r(thigh_rot_otr_r:double=MapRangeClamped(Value=@n10.Pitch, InRangeA=20, InRangeB=50)) -> (Output_Get:double)
   n11 = GetOwnerSkeletalMeshComponent()
   n12 = BreakRotator(InRot=BreakTransform(InTransform=GetDeltaTransformFromRefPose(self=@n11, BoneName=thigh_l, BaseName=pelvis)).Rotation)
   > Set thigh_rot_up_l(thigh_rot_up_l:double=MapRangeClamped(Value=@n12.Yaw, InRangeA=-10, InRangeB=-100)) -> (Output_Get:double)
   > Set thigh_rot_inr_l(thigh_rot_inr_l:double=MapRangeClamped(Value=@n12.Pitch, InRangeA=0, InRangeB=-10, OutRangeA=1, OutRangeB=0)) -> (Output_Get:double)
   > n13 = Set thigh_rot_otr_l(thigh_rot_otr_l:double=MapRangeClamped(Value=@n12.Pitch, InRangeA=20, InRangeB=50)) -> (Output_Get:double)
   Set head Rot(
    execute=[@n47.OutputPin, @n13.then]
    head Rot:double=GetAlphaFromReferenceTransformDelta(SkeletalMeshComponent=@n11, BoneName=upperarm_l, BaseName=spine_05, DeltaComponentType=EulerY, InputRange=(X=0,Y=40), TwistAxis="1,0,0")
   ) -> (Output_Get:double)
   n14 = GetDeltaTransformFromRefPose(self=GetOwnerSkeletalMeshComponent(), BoneName=calf_r, BaseName=pelvis)
   n15 = MapRangeClamped(Value=BreakRotator(InRot=BreakTransform(InTransform=@n14).Rotation).Yaw, InRangeA=60, InRangeB=90)
   n16 = AND Boolean(A=Get bIsCrouching(), B=NOT Boolean(A=Get bIsSliding()))
   > Set knee_bend_r(knee_bend_r:double=FInterpTo(Current=Get knee_bend_r(), Target=SelectFloat(A=@n15, B=0, bPickA=@n16), DeltaTime=1, InterpSpeed=1)) -> (Output_Get:double)
   n17 = GetOwnerSkeletalMeshComponent()
   n18 = BreakRotator(InRot=BreakTransform(InTransform=GetDeltaTransformFromRefPose(self=@n17, BoneName=calf_l, BaseName=pelvis)).Rotation)
   n19 = NOT Boolean(A=Get bIsSliding())
   n20 = Get bIsCrouching()
   > Set knee_bend_l(knee_bend_l:double=SelectFloat(A=MapRangeClamped(Value=@n18.Yaw, InRangeA=60, InRangeB=90), B=0, bPickA=AND Boolean(A=@n20, B=@n19))) -> (Output_Get:double)
   n21 = NOT Boolean(A=Get bIsSliding())
   n22 = Get bIsSprinting()
   > Set sprinting(sprinting:double=SelectFloat(A=1, B=0, bPickA=AND Boolean(A=@n22, B=@n21))) -> (Output_Get:double)
   > Set sliding(sliding:double=SelectFloat(A=1, B=0, bPickA=Get bIsSliding())) -> (Output_Get:double)
   n23 = Branch(execute=@n2.then_2, Condition=Get bIsSwimSprinting() -> (bIsSwimSprinting:bool))
   > Set GravityOverride_b(GravityOverride_b:Vector="0, 0, 0") -> (Output_Get:Vector)
   n24 = Branch(execute=@n23.else, Condition=Get bIsSkyDiving())
   > n25 = Branch(Condition=Get bIsParachuteOpened())
   Set GravityOverride_b(execute=[@n25.then, @n24.else], GravityOverride_b="0, 0,-920")
   n26 = Branch(execute=@n25.else, Condition="float > float"(A=Get PawnVelocityZ(), B=50) {is=title K2Node_CallFunction})
   > Set GravityOverride_b(GravityOverride_b="0, 0,200")
   n27 = Get Skel_Mesh()
   n28 = GetSocketRotation!(self:Scene Component Object Reference=@n27, InSocketName:Name=pelvis) -> (ReturnValue:Rotator)
   n29 = PerlinNoise1D!(Value:float=GetGameTimeInSeconds()) -> (ReturnValue:float)
   n30 = float * float(A="Sin (Degrees)"!(A:double=float * float(A=@n29, B=250)) -> (ReturnValue:double), B=-100)
   n31 = vector + vector(
    A:Vector=vector * float(A=GetForwardVector!(InRot:Rotator=@n28) -> (ReturnValue:Vector), B=-800)
    B:Vector=MakeVector(Y=0, Z=float + float(A=@n30, B=200) {is=title K2Node_CommutativeAssociativeBinaryOperator})
   ) -> (ReturnValue:Vector) {is=title K2Node_CommutativeAssociativeBinaryOperator}
   Set GravityOverride_b(execute=@n26.else, GravityOverride_b=@n31)
   n32 = Cast To FortPlayerPawn(execute=@n2.then_4, Object=GetOwningActor())
   > Set As Fort Player Pawn(As Fort Player Pawn:Fort Player Pawn Object Reference=@n32) -> (Output_Get:Fort Player Pawn Object Reference)
   > n33 = Get!(execute:Exec) -> (then:Exec, else:Exec, As Fort Player Pawn:Fort Player Pawn Object Reference)
   > Set is_wearing_westsausage(is_wearing_westsausage:bool=HasMatchingGameplayTag!(self:Gameplay Tag Asset Interface Interface=@n33, TagToCheck:Gameplay Tag Structure=(TagName="Granted.Athena.HandsAreReplaced")) -> (ReturnValue:bool)) -> (Output_Get:bool)
   n34 = Get Dynamic Material Instance(execute=@n2.then_5, SkeletalMeshComponent=GetPartSkeletalMeshComponent(), ElementIndex=1)
   > n35 = Branch(Condition=Get is_wearing_westsausage() -> (is_wearing_westsausage:bool))
   > n36 = Branch(Condition=Get is_performing_emote() -> (is_performing_emote:bool))
   SetScalarParameterValue(execute=[@n39.then, @n36.then, @n38.then, @n35.else, @n37.then], self=@n34, ParameterName=Hide Element 09, Value=0)
   n37 = Branch(execute=@n36.else, Condition=Get bIsValetDriver() -> (bIsValetDriver:bool))
   n38 = Branch(execute=@n37.else, Condition=Get bIsMeatballDriver() -> (bIsMeatballDriver:bool))
   n39 = Branch(execute=@n38.else, Condition=Get bIsSkyDiving())
   SetScalarParameterValue(execute=@n39.else, self=@n34, ParameterName=Hide Element 09, Value=1)
   n40 = Branch(execute=@n2.then_3, Condition=Get bIsOstrichVehicleDriver() -> (bIsOstrichVehicleDriver:bool))
   Set in vehicle rigid alpha(execute=[@n42.then, @n40.then, @n41.then, @n43.then], in vehicle rigid alpha:double=0) -> (Output_Get:double)
   > Set in vehicle scale alpha(in vehicle scale alpha:double=1) -> (Output_Get:double)
   n41 = Branch(execute=@n40.else, Condition=Get bIsFerretVehicleDriver() -> (bIsFerretVehicleDriver:bool))
   n42 = Branch(execute=@n41.else, Condition=Get bIsOctopusVehicleDriver() -> (bIsOctopusVehicleDriver:bool))
   n43 = Branch(execute=@n42.else, Condition=Get bIsValetDriver())
   Set in vehicle rigid alpha(execute=@n43.else, in vehicle rigid alpha=1)
   > Set in vehicle scale alpha(in vehicle scale alpha=0)
   Event OnCharacterCustomizationLoaded()
   > Set skelmesh(skelmesh:Skeletal Mesh Component Object Reference=GetOwnerSkeletalMeshComponent()) -> (Output_Get:Skeletal Mesh Component Object Reference)
   Event BlueprintInitializeAnimation()
   n44 = GetOwnerSkeletalMeshComponent()
   > Is Valid(InputObject=@n44)
   > Set Skel_Mesh(Skel_Mesh=@n44)
   n45 = TryGetPawnOwner()
   > Is Valid(InputObject=@n45)
   > n46 = Cast To FortPlayerPawnAthena(Object=@n45)
   > Set PlayerPawn(PlayerPawn:Fort Player Pawn Athena Object Reference=@n46) -> (Output_Get:Fort Player Pawn Athena Object Reference)
   Event OnBeginPlayingEmote() {is=title K2Node_Event}
   > Set is_performing_emote(is_performing_emote:bool=true) -> (Output_Get:bool)
   Event OnEndPlayingEmote() {is=title K2Node_Event}
   > Set is_performing_emote(is_performing_emote=false)
   n47 = Reroute Node()
   n48 = Do Once()
   Calculate Velocity From Position History!(
    execute
    DeltaSeconds:float=0
    Position:Vector="0, 0, 0"
    History:"Position History Structure (by ref)"
    NumberOfSamples:int=16
    VelocityMin:float=0.f
    VelocityMax:float=128.f
   ) -> (then, ReturnValue:float)
   "Equal (Object)"(A=@n5, B=/BRCosmetics/Accessories/FORT_Backpacks/M_MED_Ruckus/Meshes/M_MED_Ruckus_Pack) {is=title K2Node_CallFunction}
   GetVectorArrayAverage!(Vectors:Array of Vectors=Make Array!("[0]":Vector=@n6.LeftFootSlopeTranslation, "[1]":Vector=@n6.RightFootSlopeTranslation) -> (Array:Array of Vectors)) -> (ReturnValue:Vector)
   n49 = GetSocketLocation!(self:Scene Component Object Reference=@n27, InSocketName:Name=pelvis) -> (ReturnValue:Vector)
   DrawDebugLine!(
    execute
    LineStart:Vector=@n49
    LineEnd:Vector=vector + vector(A=@n49, B=@n31)
    LineColor:Linear Color Structure=(R=1,G=0,B=0,A=1)
    Duration:float=0
    Thickness:float=3
    DepthPriority:EDrawDebugSceneDepthPriorityGroup Enum=World
   ) -> (then)
   OR Boolean(A=@n16, B=Get is_performing_emote())
  ]}
]
variables [
 As Fort Player Pawn {type=FortPlayerPawn}
 GravityOverride_b {type=Vector}
 head Rot {type=number}
 in vehicle rigid alpha {type=number}
 in vehicle scale alpha {type=number}
 is_performing_emote {type=boolean}
 is_wearing_westsausage {type=boolean}
 knee_bend_l {type=number}
 knee_bend_r {type=number}
 PlayerPawn {type=FortPlayerPawnAthena}
 Ruckus_pack {type=boolean}
 Skel_Mesh {type=SkeletalMeshComponent}
 skelmesh {type=SkeletalMeshComponent}
 sliding {type=number}
 slope_foot_rot {type=Rotator}
 slope_foot_trans {type=Vector}
 sprinting {type=number}
 thigh_rot_inr_l {type=number}
 thigh_rot_inr_r {type=number}
 thigh_rot_otr_l {type=number}
 thigh_rot_otr_r {type=number}
 thigh_rot_up_l {type=number}
 thigh_rot_up_r {type=number}
]
