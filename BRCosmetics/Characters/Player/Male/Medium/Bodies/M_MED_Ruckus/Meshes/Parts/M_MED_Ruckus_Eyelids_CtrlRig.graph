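# Control Rig for the M_MED_Ruckus character's eyelids that drives procedural lid bone chains from the face rig's eye and lid bones.
# Differences in X rotation between lid and eye bones become clamped upper and lower values per side, with blink_limit reducing upper-lid overlap past the lower.
# Each value rotates its main bone near one extreme and slides seven bones along Z, scaling XY by cos(asin(Z)) to follow a spherical profile.
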ControlRigBlueprint /BRCosmetics/Characters/Player/Male/Medium/Bodies/M_MED_Ruckus/Meshes/Parts/M_MED_Ruckus_Eyelids_CtrlRig
graphs [
 RigVMFunctionLibrary {links [], nodes []}
 RigVMModel {nodes [
   RigUnit_BeginExecution = &RigUnit_BeginExecution()
   n1 = /Script/ControlRig.RigUnit_SequenceExecution(ExecuteContext=@RigUnit_BeginExecution.ExecutePin) {is=struct RigVMUnitNode}
   n2 = RigVMVariableNode() {variable=dv_lwr_L}
   &RigUnit_OffsetTransformForItem(
    Item=(Type=Bone,Name="L_lid_lwr_main")
    OffsetTransform=(Rotation=(X=0,Y=0,Z=0,W=1),Translation=(X=0,Y=0,Z=0),Scale3D=(X=1,Y=1,Z=1))
    Weight=1
    bPropagateToChildren=True
    OffsetTransform.Rotation=&RigVMFunction_MathQuaternionFromEuler(
     Euler=(X=50,Y=0,Z=0)
     Euler.X=&RigVMFunction_MathFloatRemap(SourceMaximum=1, SourceMinimum=0.8, TargetMaximum=-30, TargetMinimum=0, bClamp=true, Value=@n2.Value).Result
    ).Result
    ExecutePin=@n1.D
   )
   > n3 = /Script/RigVM.RigVMFunction_ForLoopCount(Count=7) {is=struct RigVMUnitNode}
   n4 = /Script/ControlRig.RigUnit_CollectionItems(Items=((Type=Bone,Name="L_lid_lwr_1"),(Type=Bone,Name="L_lid_lwr_2"),(Type=Bone,Name="L_lid_lwr_3"),(Type=Bone,Name="L_lid_lwr_4"),(Type=Bone,Name="L_lid_lwr_5"),(Type=Bone,Name="L_lid_lwr_6"),(Name="L_lid_lwr_mid"))) {is=struct RigVMUnitNode}
   n5 = /Script/ControlRig.RigUnit_CollectionItemAtIndex(Index=@n3.Index, Collection=@n4.Collection) {is=struct RigVMUnitNode}
   n6 = RigVMDispatchNode(Values=(0.195,0.383,0.555,0.707,0.831,0.924,0), Index=@n3.Index) {template="DISPATCH_RigVMDispatch_SelectInt32(in Index,in Values,out Result)"}
   n7 = &RigVMFunction_MathFloatMul(B=-1, A=@n6.Result) {template="Multiply::Execute(in A,in B,out Result)"}
   n8 = &RigVMFunction_MathFloatAdd(
    B=&RigVMFunction_MathFloatMul(
     A=&RigVMFunction_MathFloatAdd(A=1, B=@n7.Result) {template="Add::Execute(in A,in B,out Result)"}.Result
     B=RigVMVariableNode() {variable=dv_lwr_L}.Value
    ) {template="Multiply::Execute(in A,in B,out Result)"}.Result
    A=&RigVMFunction_MathFloatMul(B=1, A=@n6.Result) {template="Multiply::Execute(in A,in B,out Result)"}.Result
   ) {template="Add::Execute(in A,in B,out Result)"}
   > /Script/ControlRig.RigUnit_SetTranslation(Item=(Type=Bone), Space=LocalSpace, Value=(X=0,Y=0,Z=0), Weight=1, bPropagateToChildren=True, Item.Name=@n5.Item.Name, Value.Z=@n8.Result) {is=struct RigVMUnitNode, template="Set Transform::Execute(in Item,in Space,in bInitial,in Value,in Weight,in bPropagateToChildren)"}
   n9 = /Script/RigVM.RigVMFunction_MathFloatAsin(Value=@n8.Result) {is=struct RigVMUnitNode, template="Asin::Execute(in Value,out Result)"}
   n10 = /Script/RigVM.RigVMFunction_MathFloatMax(
    A=0.05
    B=&RigVMFunction_MathFloatMul(
     A=/Script/RigVM.RigVMFunction_MathFloatCos(Value=@n9.Result) {is=struct RigVMUnitNode, template="Cos::Execute(in Value,out Result)"}.Result
     B=RigVMDispatchNode(Values=(1.02,1.08,1.2,1.41,1.8,2.61,1), Index=@n3.Index) {template="DISPATCH_RigVMDispatch_SelectInt32(in Index,in Values,out Result)"}.Result
    ) {template="Multiply::Execute(in A,in B,out Result)"}.Result
   ) {is=struct RigVMUnitNode, template="Maximum::Execute(in A,in B,out Result)"}
   > /Script/ControlRig.RigUnit_SetScale(
    Item=(Type=Bone)
    Scale=(X=1,Y=1,Z=1)
    Space=LocalSpace
    Weight=1
    bPropagateToChildren=True
    Item.Name=@n5.Item.Name
    Scale.X=@n10.Result
    Scale.Y=@n10.Result
   ) {is=struct RigVMUnitNode}
   RigUnit_GetTransform = &RigUnit_GetTransform(Item=(Type=Bone,Name="L_eye"), Space=LocalSpace)
   n11 = &RigVMFunction_MathFloatAdd(
    A=&RigVMFunction_MathFloatMul(B=-1, A=&RigVMFunction_MathQuaternionToEuler(RotationOrder=ZYX, Value=@RigUnit_GetTransform.Transform.Rotation).Result.X) {template="Multiply::Execute(in A,in B,out Result)"}.Result
    B=&RigVMFunction_MathQuaternionToEuler(
     RotationOrder=ZYX
     Value=&RigUnit_GetTransform(Item=(Type=Bone,Name="L_eye_lid_upper_mid"), Space=LocalSpace).Transform.Rotation
    ).Result.X
   ) {template="Add::Execute(in A,in B,out Result)"}
   n12 = /Script/RigVM.RigVMFunction_MathFloatClamp(
    Maximum=1
    Minimum=-1
    Value=&RigVMFunction_MathFloatAdd(
     A=-0.5
     B=/Script/RigVM.RigVMFunction_MathFloatDiv(B=20, A=@n11.Result) {is=struct RigVMUnitNode, template="Divide::Execute(in A,in B,out Result)"}.Result
    ) {template="Add::Execute(in A,in B,out Result)"}.Result
   ) {is=struct RigVMUnitNode, template="Clamp::Execute(in Value,in Minimum,in Maximum,out Result)"}
   RigVMVariableNode(Value=@n12.Result, ExecuteContext=@n1.A) {variable=dv_upr_L}
   n13 = &RigUnit_GetTransform(Item=(Type=Bone,Name="L_eye"), Space=LocalSpace)
   n14 = &RigVMFunction_MathFloatAdd(
    A=&RigVMFunction_MathFloatMul(B=-1, A=&RigVMFunction_MathQuaternionToEuler(RotationOrder=ZYX, Value=@n13.Transform.Rotation).Result.X) {template="Multiply::Execute(in A,in B,out Result)"}.Result
    B=&RigVMFunction_MathQuaternionToEuler(
     RotationOrder=ZYX
     Value=&RigUnit_GetTransform(Item=(Type=Bone,Name="L_eye_lid_lower_mid"), Space=LocalSpace).Transform.Rotation
    ).Result.X
   ) {template="Add::Execute(in A,in B,out Result)"}
   n15 = &RigVMFunction_MathFloatClamp(
    Maximum=1
    Minimum=-1
    Value=&RigVMFunction_MathFloatAdd(A=1, B=&RigVMFunction_MathFloatDiv(B=45, A=@n14.Result) {template="Divide::Execute(in A,in B,out Result)"}.Result) {template="Add::Execute(in A,in B,out Result)"}.Result
   ) {template="Clamp::Execute(in Value,in Minimum,in Maximum,out Result)"}
   > n16 = RigVMVariableNode(Value=@n15.Result) {variable=dv_lwr_L}
   n17 = &RigVMFunction_ControlFlowBranch(Condition=RigVMVariableNode() {variable=blink_limit}.Value, ExecuteContext=@n16.ExecuteContext)
   n18 = /Script/RigVM.RigVMFunction_MathFloatSub(B=@n12.Result, A=@n15.Result) {is=struct RigVMUnitNode, template="Subtract::Execute(in A,in B,out Result)"}
   n19 = /Script/RigVM.RigVMFunction_MathFloatMin(A=@n18.Result) {is=struct RigVMUnitNode, template="Minimum::Execute(in A,in B,out Result)"}
   RigVMVariableNode(
    Value=&RigVMFunction_MathFloatAdd(A=&RigVMFunction_MathFloatMul(B=0.97, A=@n19.Result) {template="Multiply::Execute(in A,in B,out Result)"}.Result, B=@n12.Result) {template="Add::Execute(in A,in B,out Result)"}.Result
    ExecuteContext=@n17.True
   ) {variable=dv_upr_L}
   > RigVMVariableNode(Value=&RigVMFunction_MathFloatAdd(A=&RigVMFunction_MathFloatMul(B=0, A=@n19.Result) {template="Multiply::Execute(in A,in B,out Result)"}.Result, B=@n15.Result) {template="Add::Execute(in A,in B,out Result)"}.Result) {variable=dv_lwr_L}
   n20 = RigVMVariableNode() {variable=dv_upr_L}
   &RigUnit_OffsetTransformForItem(
    Item=(Type=Bone,Name="L_lid_upr_main")
    OffsetTransform=(Rotation=(X=0,Y=0,Z=0,W=1),Translation=(X=0,Y=0,Z=0),Scale3D=(X=1,Y=1,Z=1))
    Weight=1
    bPropagateToChildren=True
    OffsetTransform.Rotation=&RigVMFunction_MathQuaternionFromEuler(
     Euler=(X=50,Y=0,Z=0)
     Euler.X=&RigVMFunction_MathFloatRemap(SourceMaximum=-1, SourceMinimum=-0.8, TargetMaximum=30, TargetMinimum=0, bClamp=true, Value=@n20.Value).Result
    ).Result
    ExecutePin=@n1.C
   )
   > n21 = &RigVMFunction_ForLoopCount(Count=7)
   n22 = &RigUnit_CollectionItems(Items=((Type=Bone,Name="L_lid_upr_1"),(Type=Bone,Name="L_lid_upr_2"),(Type=Bone,Name="L_lid_upr_3"),(Type=Bone,Name="L_lid_upr_4"),(Type=Bone,Name="L_lid_upr_5"),(Type=Bone,Name="L_lid_upr_6"),(Name="L_lid_upr_mid")))
   n23 = &RigUnit_CollectionItemAtIndex(Index=@n21.Index, Collection=@n22.Collection)
   n24 = &RigVMFunction_MathFloatMul(
    B=-1
    A=RigVMDispatchNode(Values=(0.195,0.383,0.555,0.707,0.831,0.924,0), Index=@n21.Index) {template="DISPATCH_RigVMDispatch_SelectInt32(in Index,in Values,out Result)"}.Result
   ) {template="Multiply::Execute(in A,in B,out Result)"}
   n25 = &RigVMFunction_MathFloatAdd(
    B=&RigVMFunction_MathFloatMul(
     A=&RigVMFunction_MathFloatAdd(A=1, B=@n24.Result) {template="Add::Execute(in A,in B,out Result)"}.Result
     B=RigVMVariableNode() {variable=dv_upr_L}.Value
    ) {template="Multiply::Execute(in A,in B,out Result)"}.Result
    A=@n24.Result
   ) {template="Add::Execute(in A,in B,out Result)"}
   > &RigUnit_SetTranslation(Item=(Type=Bone), Space=LocalSpace, Value=(X=0,Y=0,Z=0), Weight=1, bPropagateToChildren=True, Item.Name=@n23.Item.Name, Value.Z=@n25.Result) {template="Set Transform::Execute(in Item,in Space,in bInitial,in Value,in Weight,in bPropagateToChildren)"}
   n26 = &RigVMFunction_MathFloatAsin(Value=@n25.Result) {template="Asin::Execute(in Value,out Result)"}
   n27 = &RigVMFunction_MathFloatMax(
    A=0.05
    B=&RigVMFunction_MathFloatMul(
     A=&RigVMFunction_MathFloatCos(Value=@n26.Result) {template="Cos::Execute(in Value,out Result)"}.Result
     B=RigVMDispatchNode(Values=(1.02,1.08,1.2,1.41,1.8,2.61,1), Index=@n21.Index) {template="DISPATCH_RigVMDispatch_SelectInt32(in Index,in Values,out Result)"}.Result
    ) {template="Multiply::Execute(in A,in B,out Result)"}.Result
   ) {template="Maximum::Execute(in A,in B,out Result)"}
   > &RigUnit_SetScale(
    Item=(Type=Bone)
    Scale=(X=1,Y=1,Z=1)
    Space=LocalSpace
    Weight=1
    bPropagateToChildren=True
    Item.Name=@n23.Item.Name
    Scale.X=@n27.Result
    Scale.Y=@n27.Result
   )
   n28 = &RigUnit_BeginExecution()
   n29 = &RigUnit_SequenceExecution(ExecuteContext=@n28.ExecutePin)
   n30 = &RigUnit_GetTransform(Item=(Type=Bone,Name="R_eye"), Space=LocalSpace)
   n31 = &RigVMFunction_MathFloatAdd(
    A=&RigVMFunction_MathFloatMul(B=-1, A=&RigVMFunction_MathQuaternionToEuler(RotationOrder=ZYX, Value=@n30.Transform.Rotation).Result.X) {template="Multiply::Execute(in A,in B,out Result)"}.Result
    B=&RigVMFunction_MathQuaternionToEuler(
     RotationOrder=ZYX
     Value=&RigUnit_GetTransform(Item=(Type=Bone,Name="R_eye_lid_upper_mid"), Space=LocalSpace).Transform.Rotation
    ).Result.X
   ) {template="Add::Execute(in A,in B,out Result)"}
   n32 = &RigVMFunction_MathFloatClamp(
    Maximum=1
    Minimum=-1
    Value=&RigVMFunction_MathFloatAdd(A=0.5, B=&RigVMFunction_MathFloatDiv(B=-20, A=@n31.Result) {template="Divide::Execute(in A,in B,out Result)"}.Result) {template="Add::Execute(in A,in B,out Result)"}.Result
   ) {template="Clamp::Execute(in Value,in Minimum,in Maximum,out Result)"}
   RigVMVariableNode(Value=@n32.Result, ExecuteContext=@n29.A) {variable=dv_upr_R}
   n33 = &RigUnit_GetTransform(Item=(Type=Bone,Name="R_eye"), Space=LocalSpace)
   n34 = &RigVMFunction_MathFloatAdd(
    A=&RigVMFunction_MathFloatMul(B=-1, A=&RigVMFunction_MathQuaternionToEuler(RotationOrder=ZYX, Value=@n33.Transform.Rotation).Result.X) {template="Multiply::Execute(in A,in B,out Result)"}.Result
    B=&RigVMFunction_MathQuaternionToEuler(
     RotationOrder=ZYX
     Value=&RigUnit_GetTransform(Item=(Type=Bone,Name="R_eye_lid_lower_mid"), Space=LocalSpace).Transform.Rotation
    ).Result.X
   ) {template="Add::Execute(in A,in B,out Result)"}
   n35 = &RigVMFunction_MathFloatClamp(
    Maximum=1
    Minimum=-1
    Value=&RigVMFunction_MathFloatAdd(A=-1, B=&RigVMFunction_MathFloatDiv(B=-45, A=@n34.Result) {template="Divide::Execute(in A,in B,out Result)"}.Result) {template="Add::Execute(in A,in B,out Result)"}.Result
   ) {template="Clamp::Execute(in Value,in Minimum,in Maximum,out Result)"}
   > n36 = RigVMVariableNode(Value=@n35.Result) {variable=dv_lwr_R}
   n37 = &RigVMFunction_ControlFlowBranch(Condition=RigVMVariableNode() {variable=blink_limit}.Value, ExecuteContext=@n36.ExecuteContext)
   n38 = &RigVMFunction_MathFloatSub(A=@n32.Result, B=@n35.Result) {template="Subtract::Execute(in A,in B,out Result)"}
   n39 = &RigVMFunction_MathFloatMin(A=@n38.Result) {template="Minimum::Execute(in A,in B,out Result)"}
   RigVMVariableNode(
    Value=&RigVMFunction_MathFloatAdd(
     A=&RigVMFunction_MathFloatMul(B=-0.97, A=@n39.Result) {template="Multiply::Execute(in A,in B,out Result)"}.Result
     B=@n32.Result
    ) {template="Add::Execute(in A,in B,out Result)"}.Result
    ExecuteContext=@n37.True
   ) {variable=dv_upr_R}
   > RigVMVariableNode(Value=&RigVMFunction_MathFloatAdd(A=&RigVMFunction_MathFloatMul(B=0, A=@n39.Result) {template="Multiply::Execute(in A,in B,out Result)"}.Result, B=@n35.Result) {template="Add::Execute(in A,in B,out Result)"}.Result) {variable=dv_lwr_R}
   n40 = RigVMVariableNode() {variable=dv_upr_R}
   &RigUnit_OffsetTransformForItem(
    Item=(Type=Bone,Name="R_lid_upr_main")
    OffsetTransform=(Rotation=(X=0,Y=0,Z=0,W=1),Translation=(X=0,Y=0,Z=0),Scale3D=(X=1,Y=1,Z=1))
    Weight=1
    bPropagateToChildren=True
    OffsetTransform.Rotation=&RigVMFunction_MathQuaternionFromEuler(
     Euler=(X=50,Y=0,Z=0)
     Euler.X=&RigVMFunction_MathFloatRemap(SourceMaximum=1, SourceMinimum=0.8, TargetMaximum=-30, TargetMinimum=0, bClamp=true, Value=@n40.Value).Result
    ).Result
    ExecutePin=@n29.C
   )
   > n41 = &RigVMFunction_ForLoopCount(Count=7)
   n42 = &RigUnit_CollectionItems(Items=((Type=Bone,Name="R_lid_upr_1"),(Type=Bone,Name="R_lid_upr_2"),(Type=Bone,Name="R_lid_upr_3"),(Type=Bone,Name="R_lid_upr_4"),(Type=Bone,Name="R_lid_upr_5"),(Type=Bone,Name="R_lid_upr_6"),(Name="R_lid_upr_mid")))
   n43 = &RigUnit_CollectionItemAtIndex(Index=@n41.Index, Collection=@n42.Collection)
   n44 = RigVMDispatchNode(Values=(0.195,0.383,0.555,0.707,0.831,0.924,0), Index=@n41.Index) {template="DISPATCH_RigVMDispatch_SelectInt32(in Index,in Values,out Result)"}
   n45 = &RigVMFunction_MathFloatMul(B=-1, A=@n44.Result) {template="Multiply::Execute(in A,in B,out Result)"}
   n46 = &RigVMFunction_MathFloatAdd(
    B=&RigVMFunction_MathFloatMul(
     A=&RigVMFunction_MathFloatAdd(A=1, B=@n45.Result) {template="Add::Execute(in A,in B,out Result)"}.Result
     B=RigVMVariableNode() {variable=dv_upr_R}.Value
    ) {template="Multiply::Execute(in A,in B,out Result)"}.Result
    A=&RigVMFunction_MathFloatMul(B=1, A=@n44.Result) {template="Multiply::Execute(in A,in B,out Result)"}.Result
   ) {template="Add::Execute(in A,in B,out Result)"}
   > &RigUnit_SetTranslation(Item=(Type=Bone), Space=LocalSpace, Value=(X=0,Y=0,Z=0), Weight=1, bPropagateToChildren=True, Item.Name=@n43.Item.Name, Value.Z=@n46.Result) {template="Set Transform::Execute(in Item,in Space,in bInitial,in Value,in Weight,in bPropagateToChildren)"}
   n47 = &RigVMFunction_MathFloatAsin(Value=@n46.Result) {template="Asin::Execute(in Value,out Result)"}
   n48 = &RigVMFunction_MathFloatMax(
    A=0.05
    B=&RigVMFunction_MathFloatMul(
     A=&RigVMFunction_MathFloatCos(Value=@n47.Result) {template="Cos::Execute(in Value,out Result)"}.Result
     B=RigVMDispatchNode(Values=(1.02,1.08,1.2,1.41,1.8,2.61,1), Index=@n41.Index) {template="DISPATCH_RigVMDispatch_SelectInt32(in Index,in Values,out Result)"}.Result
    ) {template="Multiply::Execute(in A,in B,out Result)"}.Result
   ) {template="Maximum::Execute(in A,in B,out Result)"}
   > &RigUnit_SetScale(
    Item=(Type=Bone)
    Scale=(X=1,Y=1,Z=1)
    Space=LocalSpace
    Weight=1
    bPropagateToChildren=True
    Item.Name=@n43.Item.Name
    Scale.X=@n48.Result
    Scale.Y=@n48.Result
   )
   n49 = RigVMVariableNode() {variable=dv_lwr_R}
   &RigUnit_OffsetTransformForItem(
    Item=(Type=Bone,Name="R_lid_lwr_main")
    OffsetTransform=(Rotation=(X=0,Y=0,Z=0,W=1),Translation=(X=0,Y=0,Z=0),Scale3D=(X=1,Y=1,Z=1))
    Weight=1
    bPropagateToChildren=True
    OffsetTransform.Rotation=&RigVMFunction_MathQuaternionFromEuler(
     Euler=(X=50,Y=0,Z=0)
     Euler.X=&RigVMFunction_MathFloatRemap(SourceMaximum=-1, SourceMinimum=-0.8, TargetMaximum=30, TargetMinimum=0, bClamp=true, Value=@n49.Value).Result
    ).Result
    ExecutePin=@n29.D
   )
   > n50 = &RigVMFunction_ForLoopCount(Count=7)
   n51 = &RigUnit_CollectionItems(Items=((Type=Bone,Name="R_lid_lwr_1"),(Type=Bone,Name="R_lid_lwr_2"),(Type=Bone,Name="R_lid_lwr_3"),(Type=Bone,Name="R_lid_lwr_4"),(Type=Bone,Name="R_lid_lwr_5"),(Type=Bone,Name="R_lid_lwr_6"),(Name="R_lid_lwr_mid")))
   n52 = &RigUnit_CollectionItemAtIndex(Index=@n50.Index, Collection=@n51.Collection)
   n53 = RigVMDispatchNode(Values=(0.195,0.383,0.555,0.707,0.831,0.924,0), Index=@n50.Index) {template="DISPATCH_RigVMDispatch_SelectInt32(in Index,in Values,out Result)"}
   n54 = &RigVMFunction_MathFloatMul(B=-1, A=@n53.Result) {template="Multiply::Execute(in A,in B,out Result)"}
   n55 = &RigVMFunction_MathFloatAdd(
    B=&RigVMFunction_MathFloatMul(
     A=&RigVMFunction_MathFloatAdd(A=1, B=@n54.Result) {template="Add::Execute(in A,in B,out Result)"}.Result
     B=RigVMVariableNode() {variable=dv_lwr_R}.Value
    ) {template="Multiply::Execute(in A,in B,out Result)"}.Result
    A=&RigVMFunction_MathFloatMul(B=-1, A=@n53.Result) {template="Multiply::Execute(in A,in B,out Result)"}.Result
   ) {template="Add::Execute(in A,in B,out Result)"}
   > &RigUnit_SetTranslation(Item=(Type=Bone), Space=LocalSpace, Value=(X=0,Y=0,Z=0), Weight=1, bPropagateToChildren=True, Item.Name=@n52.Item.Name, Value.Z=@n55.Result) {template="Set Transform::Execute(in Item,in Space,in bInitial,in Value,in Weight,in bPropagateToChildren)"}
   n56 = &RigVMFunction_MathFloatAsin(Value=@n55.Result) {template="Asin::Execute(in Value,out Result)"}
   n57 = &RigVMFunction_MathFloatMax(
    A=0.05
    B=&RigVMFunction_MathFloatMul(
     A=&RigVMFunction_MathFloatCos(Value=@n56.Result) {template="Cos::Execute(in Value,out Result)"}.Result
     B=RigVMDispatchNode(Values=(1.02,1.08,1.2,1.41,1.8,2.61,1), Index=@n50.Index) {template="DISPATCH_RigVMDispatch_SelectInt32(in Index,in Values,out Result)"}.Result
    ) {template="Multiply::Execute(in A,in B,out Result)"}.Result
   ) {template="Maximum::Execute(in A,in B,out Result)"}
   > &RigUnit_SetScale(
    Item=(Type=Bone)
    Scale=(X=1,Y=1,Z=1)
    Space=LocalSpace
    Weight=1
    bPropagateToChildren=True
    Item.Name=@n52.Item.Name
    Scale.X=@n57.Result
    Scale.Y=@n57.Result
   )
   &RigUnit_BeginExecution()
   &RigVMFunction_MathFloatMax(A=@n38.Result) {template="Maximum::Execute(in A,in B,out Result)"}
   &RigVMFunction_MathFloatMax(A=@n18.Result) {template="Maximum::Execute(in A,in B,out Result)"}
   /Script/RigVM.RigVMFunction_MathFloatRad(Value=@n25.Result) {is=struct RigVMUnitNode, template="Radians::Execute(in Value,out Result)"}
   &RigVMFunction_MathFloatRad(Value=@n46.Result) {template="Radians::Execute(in Value,out Result)"}
   &RigVMFunction_MathFloatRad(Value=@n55.Result) {template="Radians::Execute(in Value,out Result)"}
   &RigVMFunction_MathFloatRad(Value=@n8.Result) {template="Radians::Execute(in Value,out Result)"}
  ]}
]
hierarchy [
 root {loc=<0 0 0>, rot=<0 0 0 1>}
 pelvis {parent=root, loc=<0 -0.7071 93.1191>, rot=<0.0127 0.707 -0.0127 -0.707>}
 spine_01 {parent=pelvis, loc=<6.7494 -0.6388 0>, rot=<0 0 0.034 -0.9994>}
 spine_02 {parent=spine_01, loc=<7.2518 -0.035 0>, rot=<0 0 0.0238 -0.9997>}
 spine_03 {parent=spine_02, loc=<6.0104 0 0>, rot=<0 0 0.0149 -0.9999>}
 spine_04 {parent=spine_03, loc=<8.1024 0 0>, rot=<0 0 -0.0132 -0.9999>}
 spine_05 {parent=spine_04, loc=<10.0146 0 0>, rot=<0 0 -0.0015 -1>}
 neck_01 {parent=spine_05, loc=<11.3138 0.1529 0>, rot=<0 0 -0.2003 -0.9797>}
 neck_02 {parent=neck_01, loc=<4.5627 0 0>, rot=<0 0 0.0504 -0.9987>}
 head {parent=neck_02, loc=<5.3914 0 0>, rot=<0 0 0.115 -0.9934>}
 hair_bottom {parent=head, loc=<0.8599 -6.1495 0>, rot=<0.3585 -0.6095 0.6095 -0.3585>}
 hair_bottom_r {parent=hair_bottom, loc=<-1.0379 -4.7429 2.1604>, rot=<0.7095 -0.2692 -0.0697 -0.6475>}
 hair_bottom_l {parent=hair_bottom, loc=<-1.0379 4.7429 2.1604>, rot=<0.6475 -0.0697 -0.2692 -0.7095>}
 L_lid_upr_main {parent=head, loc=<13.2516 7.619 -4.3704>, rot=<0.7071 0.0029 -0.7071 0.0029>, scale=<4.497 4.497 4.497>}
 L_lid_upr_end {parent=L_lid_upr_main, loc=<0 0 -1.0001>, rot=<0 0 0 1>}
 L_lid_upr_mid {parent=L_lid_upr_main, loc=<0 0 -0.5>, rot=<0 0 0 1>, scale=<0.866 0.866 1>}
 L_lid_upr_1 {parent=L_lid_upr_main, loc=<0 0 -0.5975>, rot=<0 0 0 1>, scale=<0.8179 0.8179 1>}
 L_lid_upr_2 {parent=L_lid_upr_main, loc=<0 0 -0.6913>, rot=<0 0 0 1>, scale=<0.7804 0.7804 1>}
 L_lid_upr_3 {parent=L_lid_upr_main, loc=<0 0 -0.7777>, rot=<0 0 0 1>, scale=<0.7544 0.7544 1>}
 L_lid_upr_4 {parent=L_lid_upr_main, loc=<0 0 -0.8534>, rot=<0 0 0 1>, scale=<0.7349 0.7349 1>}
 L_lid_upr_5 {parent=L_lid_upr_main, loc=<0 0 -0.9156>, rot=<0 0 0 1>, scale=<0.7236 0.7236 1>}
 L_lid_upr_6 {parent=L_lid_upr_main, loc=<0 0 -0.9619>, rot=<0 0 0 1>, scale=<0.7137 0.7137 1>}
 L_lid_lwr_main {parent=head, loc=<13.2516 7.619 -4.3704>, rot=<0.7071 0.0029 -0.7071 0.0029>, scale=<4.497 4.497 4.497>}
 L_lid_lwr_mid {parent=L_lid_lwr_main, loc=<0 0 0.5>, rot=<0 0 0 1>, scale=<0.866 0.866 1>}
 L_lid_lwr_end {parent=L_lid_lwr_main, loc=<0 0 1.0001>, rot=<0 0 0 1>}
 L_lid_lwr_1 {parent=L_lid_lwr_main, loc=<0 0 0.5975>, rot=<0 0 0 1>, scale=<0.8179 0.8179 1>}
 L_lid_lwr_2 {parent=L_lid_lwr_main, loc=<0 0 0.6913>, rot=<0 0 0 1>, scale=<0.7804 0.7804 1>}
 L_lid_lwr_3 {parent=L_lid_lwr_main, loc=<0 0 0.7777>, rot=<0 0 0 1>, scale=<0.7544 0.7544 1>}
 L_lid_lwr_4 {parent=L_lid_lwr_main, loc=<0 0 0.8534>, rot=<0 0 0 1>, scale=<0.7349 0.7349 1>}
 L_lid_lwr_5 {parent=L_lid_lwr_main, loc=<0 0 0.9156>, rot=<0 0 0 1>, scale=<0.7236 0.7236 1>}
 L_lid_lwr_6 {parent=L_lid_lwr_main, loc=<0 0 0.9619>, rot=<0 0 0 1>, scale=<0.7138 0.7138 1>}
 R_lid_upr_main {parent=head, loc=<13.2516 7.619 4.3704>, rot=<0.7071 0.0029 0.7071 -0.0029>, scale=<4.497 4.497 4.497>}
 R_lid_upr_mid {parent=R_lid_upr_main, loc=<0 0 0.5>, rot=<0 0 0 1>, scale=<0.866 0.866 1>}
 R_lid_upr_end {parent=R_lid_upr_main, loc=<0 0 1.0001>, rot=<0 0 0 1>}
 R_lid_upr_1 {parent=R_lid_upr_main, loc=<0 0 0.5975>, rot=<0 0 0 1>, scale=<0.8179 0.8179 1>}
 R_lid_upr_2 {parent=R_lid_upr_main, loc=<0 0 0.6913>, rot=<0 0 0 1>, scale=<0.7804 0.7804 1>}
 R_lid_upr_3 {parent=R_lid_upr_main, loc=<0 0 0.7777>, rot=<0 0 0 1>, scale=<0.7544 0.7544 1>}
 R_lid_upr_4 {parent=R_lid_upr_main, loc=<0 0 0.8534>, rot=<0 0 0 1>, scale=<0.7349 0.7349 1>}
 R_lid_upr_5 {parent=R_lid_upr_main, loc=<0 0 0.9156>, rot=<0 0 0 1>, scale=<0.7236 0.7236 1>}
 R_lid_upr_6 {parent=R_lid_upr_main, loc=<0 0 0.9619>, rot=<0 0 0 1>, scale=<0.7137 0.7137 1>}
 R_lid_lwr_main {parent=head, loc=<13.2516 7.619 4.3704>, rot=<0.7071 0.0029 0.7071 -0.0029>, scale=<4.497 4.497 4.497>}
 R_lid_lwr_mid {parent=R_lid_lwr_main, loc=<0 0 -0.5>, rot=<0 0 0 1>, scale=<0.866 0.866 1>}
 R_lid_lwr_end {parent=R_lid_lwr_main, loc=<0 0 -1.0001>, rot=<0 0 0 1>}
 R_lid_lwr_1 {parent=R_lid_lwr_main, loc=<0 0 -0.5975>, rot=<0 0 0 1>, scale=<0.8179 0.8179 1>}
 R_lid_lwr_2 {parent=R_lid_lwr_main, loc=<0 0 -0.6913>, rot=<0 0 0 1>, scale=<0.7804 0.7804 1>}
 R_lid_lwr_3 {parent=R_lid_lwr_main, loc=<0 0 -0.7777>, rot=<0 0 0 1>, scale=<0.7544 0.7544 1>}
 R_lid_lwr_4 {parent=R_lid_lwr_main, loc=<0 0 -0.8534>, rot=<0 0 0 1>, scale=<0.7349 0.7349 1>}
 R_lid_lwr_5 {parent=R_lid_lwr_main, loc=<0 0 -0.9156>, rot=<0 0 0 1>, scale=<0.7236 0.7236 1>}
 R_lid_lwr_6 {parent=R_lid_lwr_main, loc=<0 0 -0.9619>, rot=<0 0 0 1>, scale=<0.7138 0.7138 1>}
 faceAttach {parent=head, loc=<-3.8533 -0.7383 0>, rot=<0.0412 -0.7059 -0.0412 -0.7059>, scale=<0.89 0.89 0.89>}
 R_eye {parent=faceAttach, loc=<-4.9106 7.0889 20.1815>, rot=<0 0 0.9998 -0.0187>}
 R_eye_lid_upper_mid {parent=R_eye, loc=<-0.0396 -1.1712 0.2636>, rot=<0.0523 -0.0009 0.0167 -0.9985>}
 R_eye_lid_lower_mid {parent=R_eye, loc=<-0.0085 -0.9422 -0.7046>, rot=<0.0329 -0.0005 0.0167 -0.9993>}
 L_eye {parent=faceAttach, loc=<4.9106 7.0889 20.1815>, rot=<0 0 0.9998 0.0187>}
 L_eye_lid_upper_mid {parent=L_eye, loc=<0.0251 -1.1623 0.2388>, rot=<0.0523 0 0 -0.9986>}
 L_eye_lid_lower_mid {parent=L_eye, loc=<0.0223 -0.923 -0.7488>, rot=<0.0329 0 0 -0.9995>}
]
variables [
 blink_limit {type=bool, default=True}
 dv_lwr_L {type=double, default=0.9163}
 dv_lwr_R {type=double, default=-0.9163}
 dv_upr_L {type=double, default=-0.7997}
 dv_upr_R {type=double, default=0.7995}
]
